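# Light function material that casts scrolling cloud shadows, named for Asteria, and outputs no shadow on the mobile shading path.
# Three layers of one cloud texture get coordinates from MF_CloudLayerCoord, each with its own scale and speed under a shared GlobalSpeed.
# Near the camera, small clouds combine by max with a distance-faded large layer, then crossfade to another large layer farther out.
# Shadows fade to full light above a world-height threshold.
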Material /Game/TimeOfDay/TODM_Sequence/Materials/Sky/Clouds/LF_cloudshadow_Asteria
graph {
 properties {materialDomain=MD_LightFunction, pixelDepthOffsetMode=PDOM_Legacy, usedShadingModels=Default Lit}
 nodes [
  NamedRerouteDeclaration(None=ScalarParameter() {defaultValue=1.5, parameterName=GlobalSpeed}) {name=GlobalSpeed, nodeColor=<rgba 1 0 0.4452 1>}
  n1 = TextureObject() {texture=&/T_cloudshadow2}
  n2 = MaterialFunctionCall!(
   PreviewInMatEditor
   Size=ScalarParameter() {defaultValue=220000, parameterName=scale 3}
   Speed=Multiply(A=ScalarParameter() {defaultValue=-0.15, parameterName=Speed3}, B=NamedRerouteUsage())
   Resolution=TextureProperty(Tex=@n1)
  ) {materialFunction=/Game/TimeOfDay/LightFunction/MF_CloudLayerCoord}
  n3 = TextureSample!(UVs=@n2.UVs, Tex=@n1, Level=@n2.UVs, Apply View MipBias) {
   mipValueMode=TMVM_MipLevel
   samplerSource=SSM_Wrap_WorldGroupSettings
   texture=&T_cloudshadow2
  }
  n4 = WorldPosition()
  n5 = Distance(A=CameraPositionWS(), B=@n4.XYZ)
  n6 = ScalarParameter() {defaultValue=5000, parameterName=NearRadiusSmallCloud}
  n7 = Divide(A=Subtract(A=@n5, B=@n6), B=@n6)
  n8 = TextureObject() {texture=&T_cloudshadow2}
  n9 = MaterialFunctionCall(
   Size=ScalarParameter() {defaultValue=17500, parameterName=scale 2}
   Speed=Multiply(A=ScalarParameter() {defaultValue=-1, parameterName=Speed}, B=NamedRerouteUsage())
   Resolution=TextureProperty(Tex=@n8)
  ) {materialFunction=&MF_CloudLayerCoord}
  n10 = TextureSample(UVs=@n9.UVs, Tex=@n8, Level=@n9.UVs) {
   mipValueMode=TMVM_MipLevel
   samplerSource=SSM_Wrap_WorldGroupSettings
   texture=&T_cloudshadow2
  }
  n11 = Max(A=Multiply(A=@n3.R, B=Clamp(None=@n7)), B=@n10.R)
  n12 = Power(Base=LinearInterpolate(B=ScalarParameter() {defaultValue=0.7, parameterName=shadowdark}, Alpha=Saturate(None=@n11)) {constA=1, constB=1.01})
  n13 = TextureObject() {texture=&T_cloudshadow2}
  n14 = MaterialFunctionCall(
   Size=ScalarParameter() {defaultValue=190000, parameterName=scale 1}
   Speed=Multiply(A=ScalarParameter() {defaultValue=-0.4, parameterName=Speed2}, B=NamedRerouteUsage())
   Resolution=TextureProperty(Tex=@n13)
  ) {materialFunction=&MF_CloudLayerCoord}
  n15 = TextureSample(UVs=@n14.UVs, Tex=@n13, Level=@n14.UVs) {
   mipValueMode=TMVM_MipLevel
   samplerSource=SSM_Wrap_WorldGroupSettings
   texture=&T_cloudshadow2
  }
  n16 = Power(Base=LinearInterpolate(B=ScalarParameter() {defaultValue=0.4, parameterName=shadowdark1}, Alpha=@n15.R) {constA=1, constB=1.01})
  n17 = ScalarParameter() {defaultValue=10000, parameterName=NearRadius}
  n18 = Clamp(None=Divide(A=Subtract(A=@n5, B=@n17), B=@n17))
  n19 = Divide(
   A=Subtract(A=@n4.Z, B=ScalarParameter() {defaultValue=15000, parameterName=HeightFadeOffset})
   B=ScalarParameter() {defaultValue=10000, parameterName=HeightFadeBlend}
  )
  ShadingPathSwitch(
   Default=LinearInterpolate(A=LinearInterpolate(A=@n12, B=@n16, Alpha=@n18), B=Constant() {r=1}, Alpha=Saturate(None=@n19))
   Mobile=Constant() {r=1}
  )
 ]
}
