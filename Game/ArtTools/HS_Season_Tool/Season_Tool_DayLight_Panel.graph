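# Editor utility panel of the Season Tool for adjusting daylight: time of day, sun rotation, cloud shadows and sun/sky shadows.
# Freeze and set-time inputs send cheat console commands to the game world; sun rotation is an embedded sub-widget.
# Shadow checkboxes edit the Asteria day sequence during PIE, otherwise the global time-of-day manager's actor.
# Construct caches the sun's light function material for restoring cloud shadows; unhooked test chains remain.
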EditorUtilityWidgetBlueprint /Game/ArtTools/HS_Season_Tool/Season_Tool_DayLight_Panel
graphs [
 CheckIfHasWorld {nodes [
   CheckIfHasWorld() {is=title K2Node_FunctionEntry}
   > n1 = GetPIEWorlds!(execute, bIncludeDedicatedServer:bool=false) -> (then, ReturnValue:Array of World Object References)
   > n2 = Branch(Condition=Is Empty!(TargetArray:Array of World Object References=@n1) -> (ReturnValue:bool))
   > n3 = Set IsPlaying(IsPlaying:bool=false) -> (Output_Get:bool)
   > Return Node!(execute, HasWorld?:bool=@n3)
   n4 = Set IsPlaying(execute=@n2.else, IsPlaying=true)
   > Return Node(HasWorld?=@n4)
  ]}
 EventGraph {nodes [
   n1 = GetEditorWorld(self=Get UnrealEditorSubsystem())
   > ExecuteConsoleCommand(WorldContextObject=@n1, Command=stat fps)
   n2 = GetGameWorld!(execute, self:Unreal Editor Subsystem Object Reference=Get UnrealEditorSubsystem()) -> (then, ReturnValue:World Object Reference)
   > n3 = GetActorOfClass!(
    execute
    WorldContextObject:Object Reference=@n2
    ActorClass:Actor Class Reference=/Game/TimeOfDay/TODM_Sequence/Blueprint/BP_DaySequence_Current.BP_DaySequence_Current_C
   ) -> (then, ReturnValue:BP Day Sequence Current Object Reference)
   n4 = GetSelectedActors()
   > n5 = GetActorOfClass(
    WorldContextObject="Get (a copy)"(Array:Array of Actor Object References=@n4) -> (Output:Actor Object Reference)
    ActorClass=&BP_DaySequence_Current_C
   )
   > n6 = Branch(Condition=Reroute Node(InputPin:bool) -> (OutputPin:bool))
   n7 = Get SunComponent!(self:Fort Day Sequence Actor Object Reference=@n5) -> (SunComponent:Directional Light Component Object Reference)
   > SetLightFunctionMaterial!(
    execute
    self:Light Component Object Reference=@n7
    NewLightFunctionMaterial:Material Interface Object Reference=/Game/TimeOfDay/TODM_Sequence/Materials/Sky/Clouds/LF_cloudshadow_Asteria
   ) -> (then)
   SetLightFunctionMaterial(execute=@n6.else, self=@n7)
   n8 = GetGameWorld(self=Get UnrealEditorSubsystem())
   > ExecuteConsoleCommand(
    WorldContextObject=@n8
    Command=stat fps
    SpecificPlayer=GetPlayerController!(WorldContextObject:Object Reference=@n8, PlayerIndex:int=0) -> (ReturnValue:Player Controller Object Reference)
   )
   n9 = GetSelectedActors()
   > n10 = For Each Loop(Array:Array of Object References=@n9) -> (Array Element:Object Reference)
   n11 = Get Class(Object=@n10) -> (ReturnValue:Object Class Reference)
   > LogString(InString=Append!(A="_Asset: ", B=GetDisplayName(Object=@n10), C:String=" | ", D:String="_Class: ", E:String=Get Class Display Name(Class=@n11)) -> (ReturnValue))
   n12 = GetSelectedActors()
   n13 = GetActorOfClass(
    execute=[@n12.then, @n50.LoopBody]
    WorldContextObject="Get (a copy)"(Array=@n12)
    ActorClass=&BP_DaySequence_Current_C
   )
   n14 = Get SunComponent(self=@n13)
   > SetLightFunctionMaterial(self=@n14)
   > LogString(InString=GetDisplayName(Object=Get LightFunctionMaterial!(self:Light Component Object Reference=@n14) -> (LightFunctionMaterial:Material Interface Object Reference)))
   n15 = GetPIEWorlds(bIncludeDedicatedServer=false)
   > n16 = GetActorOfClass(
    WorldContextObject="Get (a copy)"(Array:Array of World Object References=@n15) -> (Output:World Object Reference)
    ActorClass=/Game/TimeOfDay/TODM_Sequence/Blueprint/BP_Asteria_DaySequence.BP_Asteria_DaySequence_C
   ) -> (ReturnValue:BP Asteria Day Sequence Object Reference)
   > n17 = Branch()
   n18 = Get SunComponent(self=@n16)
   > SetLightFunctionMaterial(self=@n18, NewLightFunctionMaterial=&LF_cloudshadow_Asteria)
   SetLightFunctionMaterial(execute=@n17.else, self=@n18)
   n19 = CheckIfHasWorld!(execute, self) -> (then, HasWorld?:bool) {is=title K2Node_CallFunction}
   n20 = Branch(
    execute=[@n19.then, @n22.then]
    Condition=IsInPlayInEditor!(self:Level Editor Subsystem Object Reference=Get LevelEditorSubsystem()) -> (ReturnValue:bool)
   )
   > n21 = Get Asteria Day Sequencer!(In:Exec) -> (Out:Exec, AsteriaDaySequencer_ObjectRef:BP Asteria Day Sequence Object Reference) {is=title K2Node_MacroInstance}
   n22 = "On Check State Changed (CloudOnGround_CheckBox)"!() -> (OutputDelegate, then, bIsChecked:bool) {is=title K2Node_ComponentBoundEvent}
   n23 = Branch(execute=[@n21.Out, @n25.Out], Condition=@n22.bIsChecked)
   n24 = Get SunComponent(self=@n21)
   n25 = Get TODM!(In:Exec=@n20.else) -> (Out:Exec, FortDaySequence_Actor:Fort Day Sequence Actor Object Reference, Actor Object Reference:Actor Object Reference) {is=title K2Node_MacroInstance}
   n26 = Get SunComponent(self=@n25)
   SetLightFunctionMaterial(
    execute=@n23.then
    self=[@n24, @n26]
    NewLightFunctionMaterial=Get Default_LightFunctionMaterial_Ref() -> (Default_LightFunctionMaterial_Ref:Material Interface Object Reference)
   )
   SetLightFunctionMaterial(execute=@n23.else, self=[@n24, @n26])
   n27 = "OnTextChanged (Editable Text) (TimeOfDayInput)"!() -> (OutputDelegate, then, Text:"Text (by ref)") {is=title K2Node_ComponentBoundEvent}
   > n28 = GetGameWorld(self=Get UnrealEditorSubsystem())
   > n29 = Branch(Condition=TextIsEmpty(InText=@n27.Text))
   n30 = GetPlayerController(WorldContextObject=@n28, PlayerIndex=0)
   > ExecuteConsoleCommand(WorldContextObject=@n28, Command=cheat settimeofday 6, SpecificPlayer=@n30)
   ExecuteConsoleCommand(
    execute=@n29.else
    WorldContextObject=@n28
    Command=Append!(A="cheat settimeofday ", B="To String (Text)"(InText=@n27.Text)) -> (ReturnValue)
    SpecificPlayer=@n30
   )
   n31 = "On Check State Changed (FreezeTime_CheckBox)"!() -> (OutputDelegate, then, bIsChecked:bool) {is=title K2Node_ComponentBoundEvent}
   > n32 = GetGameWorld(self=Get UnrealEditorSubsystem())
   > n33 = Branch(Condition=@n31.bIsChecked)
   n34 = GetPlayerController(WorldContextObject=@n32, PlayerIndex=0)
   > ExecuteConsoleCommand(WorldContextObject=@n32, Command=Cheat SetTimeOfDaySpeed 0, SpecificPlayer=@n34)
   ExecuteConsoleCommand(execute=@n33.else, WorldContextObject=@n32, Command=Cheat SetTimeOfDaySpeed 30, SpecificPlayer=@n34)
   n35 = "On Check State Changed (SunShadows_CheckBox)"!() -> (OutputDelegate, then, bIsChecked:bool) {is=title K2Node_ComponentBoundEvent}
   > CheckIfHasWorld()
   > n36 = Branch(Condition=IsInPlayInEditor(self=Get LevelEditorSubsystem()))
   > n37 = Get Asteria Day Sequencer()
   n38 = Branch(execute=[@n37.Out, @n40.Out], Condition=@n35.bIsChecked)
   n39 = Get SunComponent(self=@n37)
   n40 = Get TODM(In=@n36.else)
   n41 = Get SunComponent(self=@n40)
   SetCastShadows!(execute=@n38.then, self:Light Component Base Object Reference=[@n39, @n41], bNewValue:bool=true) -> (then)
   SetCastShadows(execute=@n38.else, self=[@n41, @n39], bNewValue=false)
   n42 = "On Check State Changed (SkyShadows_CheckBox_1)"!() -> (OutputDelegate, then, bIsChecked:bool) {is=title K2Node_ComponentBoundEvent}
   > CheckIfHasWorld()
   > n43 = Branch(Condition=IsInPlayInEditor(self=Get LevelEditorSubsystem()))
   > n44 = Get Asteria Day Sequencer()
   n45 = Branch(execute=[@n44.Out, @n47.Out], Condition=@n42.bIsChecked)
   n46 = Get SkyLightComponent!(self:Fort Day Sequence Actor Object Reference=@n44) -> (SkyLightComponent:Sky Light Component Object Reference)
   n47 = Get TODM(In=@n43.else)
   n48 = Get SkyLightComponent(self=@n47)
   SetCastShadows(execute=@n45.then, self=[@n46, @n48], bNewValue=true)
   SetCastShadows(execute=@n45.else, self=[@n48, @n46], bNewValue=false)
   Event Construct()
   > n49 = Get TODM()
   > Set Default_LightFunctionMaterial_Ref(Default_LightFunctionMaterial_Ref:Material Interface Object Reference=Get LightFunctionMaterial(self=Get SunComponent(self=@n49))) -> (Output_Get:Material Interface Object Reference)
   n50 = For Each Loop(Array=@n12)
   Get Class Display Name(Class=Get Class(Object=@n50))
   SetLightFunctionFadeDistance!(execute, self:Light Component Object Reference, NewLightFunctionFadeDistance:float=0) -> (then)
   "To String (Object)"!(InObj:Object Reference=@n3) -> (ReturnValue:String)
   "To String (Float)"(InDouble=Get LightFunctionFadeDistance!(self:Light Component Object Reference=@n14) -> (LightFunctionFadeDistance:float))
   SetLightFunctionFadeDistance(self=@n14, NewLightFunctionFadeDistance=45)
   GetSelectedAssets()
   Event PreConstruct()
   Event Tick()
   Get Actor Class() -> (Actor Class:Actor Class Reference)
   Get Actor Class()
  ]}
 GetAsteriaDaySequencer {nodes [
   Inputs!() -> (In:Exec)
   > n1 = GetPIEWorlds(bIncludeDedicatedServer=false)
   > n2 = GetActorOfClass(
    WorldContextObject="Get (a copy)"(Array=@n1)
    ActorClass=&BP_Asteria_DaySequence_C
   )
   > Outputs!(Out:Exec, AsteriaDaySequencer_ObjectRef:BP Asteria Day Sequence Object Reference=@n2)
  ]}
 GetTODM {nodes [
   Inputs()
   > n1 = Cast To FortDaySequenceActor!(
    execute
    Object:Fort Time Of Day Manager Interface Interface=GetGlobalTimeOfDayManager!(WorldContextObject:Object Reference) -> (ReturnValue:Fort Time Of Day Manager Interface Interface)
   ) -> (then, CastFailed, AsFort Day Sequence Actor:Fort Day Sequence Actor Object Reference) {is=title K2Node_DynamicCast}
   > Outputs!(Out, FortDaySequence_Actor:Fort Day Sequence Actor Object Reference=@n1, Actor Object Reference:Actor Object Reference=@n1)
  ]}
]
variables [
 Actor Class {type=Class, default=&BP_DaySequence_Current_C}
 Default_LightFunctionMaterial_Ref {type=MaterialInterface, default=&LF_cloudshadow_Asteria}
 IsPlaying {type=boolean}
]
widgets {
 info {rootWidgetClass=&StackBox, widgetCount=26}
 widgets [
  StackBox_0 {widgetClassPath=&StackBox, widget=StackBox_0, properties {bOverride_Cursor=true}}
  DayLight_VerticalBox {
   widgetClassPath=&VerticalBox
   bIsVariable=true
   parent=StackBox_0
   widget=DayLight_VerticalBox
   properties {visibility=Visible}
   slot {class=StackBoxSlot, properties {verticalAlignment=VAlign_Top, size {sizeRule=Fill, value=1}}}
  }
  text_Title_DayLIght {
   widgetClassPath=&TextBlock
   parent=DayLight_VerticalBox
   widget=text_Title_DayLIght
   properties {
    text=Day Light
    font {
     bForceMonospaced=false
     bMaterialIsStencil=false
     fontMaterial=null
     fontObject=&Roboto
     letterSpacing=0
     monospacedWidth=1
     size=12
     skewAmount=0
     typefaceFontName=Bold
     outlineSettings {
      bApplyOutlineToDropShadows=false
      bMiteredCorners=false
      bSeparateFillAlpha=false
      outlineMaterial=null
      outlineSize=0
      outlineColor=<rgba 0 0 0 1>
     }
    }
   }
   slot {class=VerticalBoxSlot, properties {verticalAlignment=VAlign_Top}}
  }
  Spacer_2 {
   widgetClassPath=&Spacer
   parent=DayLight_VerticalBox
   widget=Spacer_2
   properties {size=<1 15>}
   slot {class=VerticalBoxSlot, properties {verticalAlignment=VAlign_Top}}
  }
  FreezeTime_HorizontalBox {
   widgetClassPath=&HorizontalBox
   parent=DayLight_VerticalBox
   widget=FreezeTime_HorizontalBox
   slot {class=VerticalBoxSlot, properties {horizontalAlignment=HAlign_Left, verticalAlignment=VAlign_Top}}
  }
  txt_SpawnSeasonAssets_5 {
   widgetClassPath=&TextBlock
   parent=FreezeTime_HorizontalBox
   widget=txt_SpawnSeasonAssets_5
   properties {
    text="(only PIE) Freeze Time"
    visibility=SelfHitTestInvisible
    font {
     bForceMonospaced=false
     bMaterialIsStencil=false
     fontMaterial=null
     fontObject=&Roboto
     letterSpacing=0
     monospacedWidth=1
     size=9
     skewAmount=0
     typefaceFontName=Bold
     outlineSettings {
      bApplyOutlineToDropShadows=false
      bMiteredCorners=false
      bSeparateFillAlpha=false
      outlineMaterial=null
      outlineSize=0
      outlineColor=<rgba 0 0 0 1>
     }
    }
   }
   slot {class=HorizontalBoxSlot, properties {horizontalAlignment=HAlign_Center, verticalAlignment=VAlign_Center, padding=<ltrb 4 2 4 2>}}
  }
  FreezeTime_CheckBox {
   widgetClassPath=&EditorUtilityCheckBox
   bIsVariable=true
   parent=FreezeTime_HorizontalBox
   widget=FreezeTime_CheckBox
   slot {class=HorizontalBoxSlot}
  }
  Spacer {
   widgetClassPath=&Spacer
   parent=DayLight_VerticalBox
   widget=Spacer
   properties {size=<1 10>}
   slot {class=VerticalBoxSlot, properties {verticalAlignment=VAlign_Top}}
  }
  FreezeTime_HorizontalBox_1 {
   widgetClassPath=&HorizontalBox
   parent=DayLight_VerticalBox
   widget=FreezeTime_HorizontalBox_1
   slot {class=VerticalBoxSlot, properties {horizontalAlignment=HAlign_Left, verticalAlignment=VAlign_Top}}
  }
  txt_SpawnSeasonAssets {
   widgetClassPath=&TextBlock
   parent=FreezeTime_HorizontalBox_1
   widget=txt_SpawnSeasonAssets
   properties {
    text="(only PIE) Set Time of day to"
    visibility=SelfHitTestInvisible
    font {
     bForceMonospaced=false
     bMaterialIsStencil=false
     fontMaterial=null
     fontObject=&Roboto
     letterSpacing=0
     monospacedWidth=1
     size=9
     skewAmount=0
     typefaceFontName=Bold
     outlineSettings {
      bApplyOutlineToDropShadows=false
      bMiteredCorners=false
      bSeparateFillAlpha=false
      outlineMaterial=null
      outlineSize=0
      outlineColor=<rgba 0 0 0 1>
     }
    }
   }
   slot {class=HorizontalBoxSlot, properties {horizontalAlignment=HAlign_Center, verticalAlignment=VAlign_Center, padding=<ltrb 4 2 4 2>}}
  }
  TimeOfDayInput {
   widgetClassPath=/Script/UMG.EditableText
   bIsVariable=true
   parent=FreezeTime_HorizontalBox_1
   widget=TimeOfDayInput
   properties {
    hintText=0600
    justification=Center
    toolTipText=Set time like 0600 or 6:00a
    widgetStyle {
     backgroundImageComposing {
      bIsDynamicallyLoaded=false
      drawAs=Border
      imageType=FullColor
      mirroring=NoMirror
      resourceName=../../../Engine/Content/Slate/Old/HyperlinkDotted.png
      resourceObject=null
      tiling=Both
      imageSize=<0 0>
      margin=<ltrb 0 0 0 0.1875>
      outlineSettings {
       bUseBrushTransparency=false
       roundingType=HalfHeightRadius
       width=0
       color {colorUseRule=UseColor_Specified, specifiedColor=<rgba 0 0 0 0>}
       cornerRadii=<0 0 0 0>
      }
      tintColor {colorUseRule=UseColor_Specified, specifiedColor=<rgba 1 1 1 1>}
      uVRegion {bIsValid=false, max=<0 0>, min=<0 0>}
     }
     backgroundImageSelected {
      bIsDynamicallyLoaded=false
      drawAs=Image
      imageType=NoImage
      mirroring=NoMirror
      resourceName=None
      resourceObject=null
      tiling=NoTile
      imageSize=<0 0>
      margin=<ltrb 0 0 0 0>
      outlineSettings {
       bUseBrushTransparency=false
       roundingType=HalfHeightRadius
       width=0
       color {colorUseRule=UseColor_Specified, specifiedColor=<rgba 0 0 0 0>}
       cornerRadii=<0 0 0 0>
      }
      tintColor {colorUseRule=UseColor_Specified, specifiedColor=<rgba 0 0.162 0.7454 1>}
      uVRegion {bIsValid=false, max=<0 0>, min=<0 0>}
     }
     caretImage {
      bIsDynamicallyLoaded=false
      drawAs=Image
      imageType=NoImage
      mirroring=NoMirror
      resourceName=None
      resourceObject=null
      tiling=NoTile
      imageSize=<0 0>
      margin=<ltrb 0 0 0 0>
      outlineSettings {
       bUseBrushTransparency=false
       roundingType=HalfHeightRadius
       width=0
       color {colorUseRule=UseColor_Specified, specifiedColor=<rgba 0 0 0 0>}
       cornerRadii=<0 0 0 0>
      }
      tintColor {colorUseRule=UseColor_Specified, specifiedColor=<rgba 1 1 1 1>}
      uVRegion {bIsValid=false, max=<0 0>, min=<0 0>}
     }
     colorAndOpacity {colorUseRule=UseColor_Specified, specifiedColor=<rgba 1 1 1 1>}
     font {
      bForceMonospaced=false
      bMaterialIsStencil=false
      fontMaterial=null
      fontObject=null
      letterSpacing=0
      monospacedWidth=1
      size=10.5
      skewAmount=0
      typefaceFontName=Regular
      outlineSettings {
       bApplyOutlineToDropShadows=false
       bMiteredCorners=false
       bSeparateFillAlpha=false
       outlineMaterial=null
       outlineSize=1
       outlineColor=<rgba 0 0 0 1>
      }
     }
     hintColorAndOpacity {colorUseRule=UseColor_Foreground, specifiedColor=<rgba 1 0 1 1>}
     selectedBackgroundColor {colorUseRule=UseColor_Foreground, specifiedColor=<rgba 1 0 1 1>}
    }
   }
   slot {class=HorizontalBoxSlot}
  }
  HorizontalBox_0 {
   widgetClassPath=&HorizontalBox
   parent=DayLight_VerticalBox
   widget=HorizontalBox_0
   slot {class=VerticalBoxSlot, properties {verticalAlignment=VAlign_Top}}
  }
  txt_SetSunRotation {
   widgetClassPath=&TextBlock
   parent=HorizontalBox_0
   widget=txt_SetSunRotation
   properties {
    text=Set Sun Rotation
    visibility=SelfHitTestInvisible
    font {
     bForceMonospaced=false
     bMaterialIsStencil=false
     fontMaterial=null
     fontObject=&Roboto
     letterSpacing=0
     monospacedWidth=1
     size=9
     skewAmount=0
     typefaceFontName=Bold
     outlineSettings {
      bApplyOutlineToDropShadows=false
      bMiteredCorners=false
      bSeparateFillAlpha=false
      outlineMaterial=null
      outlineSize=0
      outlineColor=<rgba 0 0 0 1>
     }
    }
   }
   slot {class=HorizontalBoxSlot, properties {horizontalAlignment=HAlign_Center, verticalAlignment=VAlign_Center, padding=<ltrb 4 2 4 2>}}
  }
  Season_Tool_SunRotation {
   widgetClassPath=&/Season_Tool_SunRotation
   bIsVariable=true
   parent=HorizontalBox_0
   widget=Season_Tool_SunRotation
   slot {class=HorizontalBoxSlot, properties {size {sizeRule=Fill, value=1}}}
  }
  Spacer_1 {
   widgetClassPath=&Spacer
   parent=DayLight_VerticalBox
   widget=Spacer_1
   properties {size=<1 10>}
   slot {class=VerticalBoxSlot, properties {verticalAlignment=VAlign_Top}}
  }
  FreezeTime_HorizontalBox_2 {
   widgetClassPath=&HorizontalBox
   parent=DayLight_VerticalBox
   widget=FreezeTime_HorizontalBox_2
   slot {class=VerticalBoxSlot, properties {horizontalAlignment=HAlign_Left, verticalAlignment=VAlign_Top}}
  }
  txt_SpawnSeasonAssets_1 {
   widgetClassPath=&TextBlock
   parent=FreezeTime_HorizontalBox_2
   widget=txt_SpawnSeasonAssets_1
   properties {
    text=Cloud Shadow on the Ground
    visibility=SelfHitTestInvisible
    font {
     bForceMonospaced=false
     bMaterialIsStencil=false
     fontMaterial=null
     fontObject=&Roboto
     letterSpacing=0
     monospacedWidth=1
     size=9
     skewAmount=0
     typefaceFontName=Bold
     outlineSettings {
      bApplyOutlineToDropShadows=false
      bMiteredCorners=false
      bSeparateFillAlpha=false
      outlineMaterial=null
      outlineSize=0
      outlineColor=<rgba 0 0 0 1>
     }
    }
   }
   slot {class=HorizontalBoxSlot, properties {horizontalAlignment=HAlign_Center, verticalAlignment=VAlign_Center, padding=<ltrb 4 2 4 2>}}
  }
  CloudOnGround_CheckBox {
   widgetClassPath=&EditorUtilityCheckBox
   bIsVariable=true
   parent=FreezeTime_HorizontalBox_2
   widget=CloudOnGround_CheckBox
   properties {checkedState=Checked}
   slot {class=HorizontalBoxSlot}
  }
  Spacer_3 {
   widgetClassPath=&Spacer
   parent=DayLight_VerticalBox
   widget=Spacer_3
   properties {size=<1 10>}
   slot {class=VerticalBoxSlot, properties {verticalAlignment=VAlign_Top}}
  }
  SunShadows_HorizontalBox {
   widgetClassPath=&HorizontalBox
   parent=DayLight_VerticalBox
   widget=SunShadows_HorizontalBox
   slot {class=VerticalBoxSlot, properties {horizontalAlignment=HAlign_Left, verticalAlignment=VAlign_Top}}
  }
  txt_SpawnSeasonAssets_4 {
   widgetClassPath=&TextBlock
   parent=SunShadows_HorizontalBox
   widget=txt_SpawnSeasonAssets_4
   properties {
    text=Sun Shadows
    visibility=SelfHitTestInvisible
    font {
     bForceMonospaced=false
     bMaterialIsStencil=false
     fontMaterial=null
     fontObject=&Roboto
     letterSpacing=0
     monospacedWidth=1
     size=9
     skewAmount=0
     typefaceFontName=Bold
     outlineSettings {
      bApplyOutlineToDropShadows=false
      bMiteredCorners=false
      bSeparateFillAlpha=false
      outlineMaterial=null
      outlineSize=0
      outlineColor=<rgba 0 0 0 1>
     }
    }
   }
   slot {class=HorizontalBoxSlot, properties {horizontalAlignment=HAlign_Center, verticalAlignment=VAlign_Center, padding=<ltrb 4 2 4 2>}}
  }
  SunShadows_CheckBox {
   widgetClassPath=&EditorUtilityCheckBox
   bIsVariable=true
   parent=SunShadows_HorizontalBox
   widget=SunShadows_CheckBox
   properties {checkedState=Checked}
   slot {class=HorizontalBoxSlot}
  }
  Spacer_4 {
   widgetClassPath=&Spacer
   parent=DayLight_VerticalBox
   widget=Spacer_4
   properties {size=<1 10>}
   slot {class=VerticalBoxSlot, properties {verticalAlignment=VAlign_Top}}
  }
  SkyShadows_HorizontalBox {
   widgetClassPath=&HorizontalBox
   parent=DayLight_VerticalBox
   widget=SkyShadows_HorizontalBox
   slot {class=VerticalBoxSlot, properties {horizontalAlignment=HAlign_Left, verticalAlignment=VAlign_Top}}
  }
  txt_SpawnSeasonAssets_6 {
   widgetClassPath=&TextBlock
   parent=SkyShadows_HorizontalBox
   widget=txt_SpawnSeasonAssets_6
   properties {
    text=Sky Shadows
    visibility=SelfHitTestInvisible
    font {
     bForceMonospaced=false
     bMaterialIsStencil=false
     fontMaterial=null
     fontObject=&Roboto
     letterSpacing=0
     monospacedWidth=1
     size=9
     skewAmount=0
     typefaceFontName=Bold
     outlineSettings {
      bApplyOutlineToDropShadows=false
      bMiteredCorners=false
      bSeparateFillAlpha=false
      outlineMaterial=null
      outlineSize=0
      outlineColor=<rgba 0 0 0 1>
     }
    }
   }
   slot {class=HorizontalBoxSlot, properties {horizontalAlignment=HAlign_Center, verticalAlignment=VAlign_Center, padding=<ltrb 4 2 4 2>}}
  }
  SkyShadows_CheckBox_1 {
   widgetClassPath=&EditorUtilityCheckBox
   bIsVariable=true
   parent=SkyShadows_HorizontalBox
   widget=SkyShadows_CheckBox_1
   properties {checkedState=Checked}
   slot {class=HorizontalBoxSlot}
  }
 ]
}
