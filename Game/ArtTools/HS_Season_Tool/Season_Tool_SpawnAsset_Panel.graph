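# Editor Utility Widget panel for the season art tool that spawns cosmetic items (pickaxes, backpacks, gliders, characters) into the level or PIE for review and icon rendering.
# Loaded item definitions are switched by class, then their meshes spawn with override and variant materials fetched through embedded Python.
# Spawned actors map to set indices, which place them at location markers in an AssetsSpawned folder.
# Rendering isolates each actor and saves render-target textures.
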EditorUtilityWidgetBlueprint /Game/ArtTools/HS_Season_Tool/Season_Tool_SpawnAsset_Panel
graphs [
 AssetsToFolder {nodes [
   AssetsToFolder() {is=title K2Node_FunctionEntry}
   > n1 = Keys!(
    execute
    TargetMap:Map of Actor Object References to Integers=Get AssetsSpawned_Test() -> (AssetsSpawned_Test:Map of Actor Object References to Integers)
   ) -> (then, Keys:Array of Actor Object References)
   > Execute Python Script!(
    execute
    PythonScript="import unreal\r\n\r\nactors = actors\r\nfolder_name = \"AssetsSpawned\"\r\nfolder_path =  folder_name\r\n\r\nfor actor in actors:\r\n    if isinstance(actor, unreal.Actor):\r\n        actor.set_folder_path(folder_path)"
    actors:Array of Actor Object References=@n1
   ) -> (then, ReturnValue) {inputs [actors]}
  ]}
 AssetsToLocationSpawn {nodes [
   AssetsToLocationSpawn() {is=title K2Node_FunctionEntry}
   n1 = Get SetNames() -> (SetNames:Array of Strings)
   n2 = Get LocationActors() -> (LocationActors:Array of Actor Object References)
   > n3 = Branch(Condition="integer >= integer"(A=Length(TargetArray=@n2), B=Length(TargetArray:Array of Strings=@n1)))
   n4 = Get AssetsSpawned_Test()
   > n5 = Keys(TargetMap=@n4)
   > n6 = For Each Loop(Array:Array of Actor Object References=@n5) -> (Array Element:Actor Object Reference)
   n7 = Find(TargetMap:Map of Actor Object References to Integers=@n4, Key:"Actor Object Reference (by ref)"=@n6) -> (Value:int)
   > n8 = Branch(Condition="Equal (Integer)"(A=Get valuePrevious() -> (valuePrevious:int=0), B=@n7))
   n9 = Get Actor Location!(self="Get (a copy)"(Array:Array of Actor Object References=Get LocationActors(), Dimension 1=@n7) -> (Output:Actor Object Reference)) -> (ReturnValue_X:double, ReturnValue_Y:double, ReturnValue_Z:double)
   n10 = float + float(A=Get spawnLocationOffSet() -> (spawnLocationOffSet:double), B=200)
   Set Actor Location And Rotation!(
    execute=[@n11.then, @n8.then]
    self:Actor Object Reference=@n6
    NewLocation_X:double=@n9
    NewLocation_Y:double=float + float(A=@n9.ReturnValue_Y, B=@n10)
    NewLocation_Z:double=@n9.ReturnValue_Z
    NewRotation:Rotator="0,-90,0"
    bSweep:bool=false
    bTeleport:bool=true
   ) -> (then, SweepHitResult:Hit Result Structure, ReturnValue:bool) {is=title K2Node_CallFunction}
   > Set spawnLocationOffSet(spawnLocationOffSet:double=@n10) -> (Output_Get:double)
   Set valuePrevious(execute=@n8.else, valuePrevious:int=@n7) -> (Output_Get:int=0)
   > n11 = Set spawnLocationOffSet(spawnLocationOffSet=-200)
   SpawnNewLocation(execute=@n3.else)
  ]}
 CheckForDualWield {nodes [
   n1 = CheckForDualWield!() -> (then, WeaponDefinition_Ref:Fort Weapon Melee Item Definition Object Reference) {is=title K2Node_FunctionEntry}
   > n2 = Execute Python Script!(
    execute
    PythonScript="import unreal\r\n\r\ndef get_offhand_skeletal_mesh(asset_path):\r\n    data_asset = unreal.EditorAssetLibrary.load_asset(asset_path)\r\n    if not data_asset:\r\n        unreal.log_error(f\"Failed to load asset: {asset_path}\")\r\n        return None\r\n    \r\n    weapon_asset = unreal.EditorAssetLibrary.load_asset(asset_path)\r\n    if not weapon_asset:\r\n        unreal.log_error(f\"Failed to load Weapon Definition asset: {asset_path}\")\r\n        return None\r\n    \r\n    if \"WeaponMeshOffhandOverride\" in dir(weapon_asset):\r\n        offhand_mesh = weapon_asset.get_editor_property(\"WeaponMeshOffhandOverride\")\r\n        if offhand_mesh:\r\n            unreal.log(f\"Found Offhand Skeletal Mesh: {offhand_mesh}\")\r\n            return offhand_mesh\r\n        else:\r\n            unreal.log_warning(f\"No Offhand Skeletal Mesh found in {asset_path}\")\r\n            return None\r\n\r\nasset_name = asset_name\r\nasset_path = \"/BRCosmetics/Athena/Items/Weapons/\" + asset_name + \".\" + asset_name\r\noffhand_mesh = get_offhand_skeletal_mesh(asset_path)\r\n"
    asset_name:String=GetDisplayName(Object=@n1)
   ) -> (then, ReturnValue, offhand_mesh:Skeletal Mesh Object Reference) {inputs [asset_name], outputs [offhand_mesh]}
   > n3 = Branch(Condition="Not Equal (Object)"(A:Object Reference=@n2.offhand_mesh, B:Object Reference=Get PickaxeMesh_Ref() -> (PickaxeMesh_Ref:Skeletal Mesh Object Reference)) -> (ReturnValue:bool=~) {is=title K2Node_PromotableOperator})
   > Return Node!(execute, OffHand_SkeletalMesh_Ref:Skeletal Mesh Object Reference=@n2.offhand_mesh)
   Return Node(execute=@n3.else)
  ]}
 CheckIfHasOverrideOrVariantAndSpawn {nodes [
   n1 = CheckIfHasOverrideOrVariantAndSpawn!() -> (
    then
    List of Materials:Array of Material Instance Constant Object References
    SkeletalMeshToSpawn:Skeletal Mesh Object Reference
    SetNumber:int
   ) {is=title K2Node_FunctionEntry}
   > n2 = Set ListOfMaterials(ListOfMaterials:Array of Material Instance Constant Object References=@n1) -> (Output_Get:Array of Material Instance Constant Object References)
   > n3 = Branch(Condition=Is Not Empty!(TargetArray:Array of Material Instance Constant Object References=@n2) -> (ReturnValue:bool))
   > n4 = For Each Loop(Array:Array of Material Instance Constant Object References=@n2) -> (Array Element:Material Instance Constant Object Reference)
   > LogString(InString=GetDisplayName(Object=@n4))
   > n5 = Branch(Condition=IsInPlayInEditor!(self:Level Editor Subsystem Object Reference=Get LevelEditorSubsystem()) -> (ReturnValue:bool))
   > n6 = GetPIEWorlds!(execute, bIncludeDedicatedServer:bool=false) -> (then, ReturnValue:Array of World Object References)
   > n7 = SpawnActor Season Asset Actor!(
    execute:Exec
    WorldContextObject:Object Reference="Get (a copy)"(Array:Array of World Object References=@n6) -> (Output:World Object Reference)
    Class:Actor Class Reference=&/SeasonAsset_Actor.SeasonAsset_Actor_C
    SpawnTransform_Location:Vector="0,0,2100"
    SpawnTransform_Rotation:Rotator="0, 0, 0"
    SpawnTransform_Scale:Vector="1,1,1"
    CollisionHandlingOverride:ESpawnActorCollisionHandlingMethod Enum=Undefined
    TransformScaleMethod:ESpawnActorScaleMethod Enum=MultiplyWithRoot
    Owner:Actor Object Reference
    Instigator:Pawn Object Reference
   ) -> (then:Exec, ReturnValue:Season Asset Actor Object Reference) {is=title K2Node_SpawnActorFromClass}
   > SetSkeletalMeshAsset!(
    execute
    self:Skeletal Mesh Component Object Reference=Get SeasonAsset_SkeletalMesh!(self:Season Asset Actor Object Reference=@n7) -> (SeasonAsset_SkeletalMesh:Skeletal Mesh Component Object Reference)
    NewMesh:Skeletal Mesh Object Reference=@n1.SkeletalMeshToSpawn
   ) -> (then)
   > n8 = Add!(
    execute
    TargetMap:Map of Actor Object References to Integers=Get AssetsSpawned_Test()
    Key:"Actor Object Reference (by ref)"=@n7
    Value:"Integer (by ref)"=@n1.SetNumber
   ) -> (then) {is=title K2Node_CallFunction}
   n9 = FindExistingMaterial!(execute=[@n8.then, @n23.then], BasePath:String=/BRCosmetics, MaterialFullName:String=GetDisplayName(Object=@n4), bRecursivePaths:bool=true) -> (then, OutError:Text, ReturnValue:Material Interface Object Reference)
   n10 = SpawnActorFromObject!(
    execute=@n5.else
    ObjectToUse:Object Reference=@n1.SkeletalMeshToSpawn
    Location_X:double=0
    Location_Y:double=0
    Location_Z:double=2100
    Rotation:Rotator
    bTransient:bool=false
   ) -> (then, ReturnValue:Actor Object Reference)
   n11 = GetComponentByClass!(self:Actor Object Reference=@n10, ComponentClass:Actor Component Class Reference=/Script/Engine.SkeletalMeshComponent) -> (ReturnValue:Skeletal Mesh Component Object Reference)
   SetMaterial(execute=@n9.then, self=@n11, ElementIndex=0, Material=@n9.ReturnValue)
   n12 = GetMaterials!(self:Mesh Component Object Reference=@n11) -> (ReturnValue:Array of Material Interface Object References)
   > n13 = Branch(Condition="integer >= integer"(A=Length(TargetArray:Array of Material Interface Object References=@n12), B=2))
   > n14 = SetMaterial(
    self=@n11
    ElementIndex=1
    Material="Get (a copy)"(Array:Array of Material Instance Constant Object References=@n2, Dimension 1=int + int(A=@n4.Array Index)) -> (Output:Material Instance Constant Object Reference)
   )
   AssetsToFolder(execute=[@n14.then, @n13.else]) {is=title K2Node_CallFunction}
   > n15 = Branch(Condition=Identical!(
     ArrayA:Array of Material Instance Constant Object References=Get ListOfMaterials() -> (ListOfMaterials:Array of Material Instance Constant Object References)
     ArrayB:Array of Material Instance Constant Object References=Get override_material_list() -> (override_material_list:Array of Material Instance Constant Object References)
    ) -> (ReturnValue:bool) {is=title K2Node_CallArrayFunction})
   > n16 = Set bool_HasOverride(bool_HasOverride:bool=true) -> (Output_Get:bool)
   n17 = Set bool_HasVariant(execute=@n18.then, bool_HasVariant:bool=true) -> (Output_Get:bool)
   Return Node!(execute=[@n17.then, @n16.then], HasOverride:bool=@n16, HasVariant:bool=@n17)
   n18 = Branch(execute=@n15.else, Condition=Identical(ArrayA=Get ListOfMaterials(), ArrayB=Get override_material_list()))
   LogString(execute=@n3.else, InString=NO Overrides/Variants on this list AVAILABLE)
   > AssetsToFolder()
   > n19 = Branch(Condition=Identical(ArrayA=Get ListOfMaterials(), ArrayB=Get override_material_list()))
   > n20 = Set bool_HasOverride(bool_HasOverride=false)
   n21 = Set bool_HasVariant(execute=@n22.then, bool_HasVariant=false)
   Return Node(execute=[@n21.then, @n20.then], HasOverride=@n20, HasVariant=@n21)
   n22 = Branch(execute=@n19.else, Condition=Identical(ArrayA=Get ListOfMaterials(), ArrayB=Get override_material_list()))
   n23 = Add(execute=@n10.then, TargetMap=Get AssetsSpawned_Test(), Key=@n10, Value=@n1.SetNumber)
  ]}
 ClearAssetNameArray {nodes [
   ClearAssetNameArray() {is=title K2Node_FunctionEntry}
   > Clear(TargetArray=Get AssetName() -> (AssetName:Array of Strings))
   > Clear(TargetArray=Get Asset_List() -> (Asset_List:Array of Strings))
   > Clear(TargetArray=Get SetNames())
  ]}
 ClearMaterialsList {nodes [
   ClearMaterialsList() {is=title K2Node_FunctionEntry}
   > Clear(TargetArray:Array of Material Instance Constant Object References=Get override_material_list())
   > Clear(TargetArray=Get variant_materials_list() -> (variant_materials_list:Array of Material Instance Constant Object References))
  ]}
 DeleteCurrentAssetsSpawned {nodes [
   DeleteCurrentAssetsSpawned() {is=title K2Node_FunctionEntry}
   > n1 = Keys(TargetMap=Get AssetsSpawned_Test())
   > Execute Python Script!(
    execute
    PythonScript="import unreal\r\n\r\nfolder_path = \"AssetsSpawned\"\r\nactors = actors\r\nactor_utilities_subsystem = unreal.get_editor_subsystem(unreal.EditorActorSubsystem)\r\n\r\nfor actor in actors:\r\n    if actor.get_folder_path() == folder_path:\r\n        actor_utilities_subsystem.destroy_actor(actor)"
    actors:Array of Actor Object References=@n1
   ) -> (then, ReturnValue) {inputs [actors]}
   > Clear!(execute, TargetMap:Map of Actor Object References to Integers=Get AssetsSpawned_Test()) -> (then) {is=title K2Node_CallFunction}
  ]}
 DisplaySelectedAssets {nodes [
   DisplaySelectedAssets() {is=title K2Node_FunctionEntry}
   > n1 = GetSelectedLevelActors(self=Get EditorActorSubsystem())
   > n2 = For Each Loop(Array:Array of Actor Object References=@n1) -> (Array Element:Actor Object Reference)
   Set FirstLocal(execute=@n2.Completed, FirstLocal:double=-1654) -> (Output_Get:double)
   n3 = Set FirstLocal(execute=@n2.LoopBody, FirstLocal=float - float(A=Get FirstLocal() -> (FirstLocal:double), B=-200))
   > Set Actor Location And Rotation(self=@n2, NewLocation_X=@n3, NewLocation_Y=-1578, NewLocation_Z=2072, NewRotation="0,180,0", bSweep=false, bTeleport=true)
  ]}
 EventGraph {nodes [
   n1 = Cast To AthenaCharacterItemDefinition!(execute=@n10.AthenaCharacterItemDefinition, Object=Get Array ElementSelectedAssets() -> (Array ElementSelectedAssets:Object Reference)) -> (then, CastFailed, AsAthena Character Item Definition:Athena Character Item Definition Object Reference) {is=title K2Node_DynamicCast}
   n2 = GatherAllVariantParts!(execute, ItemDefinition:Athena Character Item Definition Object Reference=@n1) -> (then, OutCharacterParts:Array of Custom Character Part Object References, ReturnValue:bool)
   > n3 = For Each Loop(Array:Array of Custom Character Part Object References=@n2) -> (Array Element:Custom Character Part Object Reference)
   > LogString(InString=GetDisplayName(Object=@n3))
   "On Clicked (btn_SpawnSeasonAssets)"() {is=title K2Node_ComponentBoundEvent}
   > CollectGarbage!(execute) -> (then)
   > ClearAssetNameArray() {is=title K2Node_CallFunction}
   > n4 = SpawnSeasonAsset()
   GetLocationActors(execute=[@n4.then, @n52.then])
   > DeleteCurrentAssetsSpawned() {is=title K2Node_CallFunction}
   > GetAssetsPathAndLoad!(execute, self, Asset_List:Array of Strings=Get Asset_List()) -> (then)
   n5 = Get ArrayOfLoadedAssets_Test() -> (ArrayOfLoadedAssets_Test:Map of Object References to Integers)
   > n6 = Keys(TargetMap:Map of Object References to Integers=@n5) -> (Keys:Array of Object References)
   > n7 = For Each Loop(Array:Array of Object References=@n6) -> (Array Element:Object Reference)
   AssetsToLocationSpawn(execute=@n7.Completed) {is=title K2Node_CallFunction}
   > Clear(TargetMap:Map of Object References to Integers=Get ArrayOfLoadedAssets_Test())
   > Set valuePrevious(valuePrevious=0) -> (Output_Get=0)
   > Set spawnLocationOffSet(spawnLocationOffSet=-200)
   > Clear(TargetArray:Array of Skeletal Mesh Object References=Get CharacterSkeletalMesh() -> (CharacterSkeletalMesh:Array of Skeletal Mesh Object References)) {is=title K2Node_CallArrayFunction}
   n8 = Set Array ElementSelectedAssets(execute=@n7.LoopBody, Array ElementSelectedAssets:Object Reference=@n7) -> (Output_Get:Object Reference)
   > Set Element_LoadAsset_SetNumber_Index(Element_LoadAsset_SetNumber_Index:int=Find(TargetMap:Map of Object References to Integers=@n5, Key:"Object Reference (by ref)"=@n7)) -> (Output_Get:int=0)
   n9 = GetDisplayName(Object=@n8)
   > Set CurrentAsset_Name(CurrentAsset_Name:String=@n9) -> (Output_Get:String)
   > Add(TargetArray=Get AssetName(), NewItem=@n9) {is=title K2Node_CallArrayFunction}
   > n10 = Switch on String!(execute:Exec, Selection:String=Get Class Display Name(Class=Get Class(Object=@n8) -> (ReturnValue:Object Class Reference))) -> (
    Default:Exec
    AthenaPickaxeItemDefinition:Exec
    AthenaBackpackItemDefinition:Exec
    AthenaGliderItemDefinition:Exec
    AthenaCharacterItemDefinition:Exec
   ) {pinNames [AthenaPickaxeItemDefinition, AthenaBackpackItemDefinition, AthenaGliderItemDefinition, AthenaCharacterItemDefinition]}
   n11 = Cast To AthenaGliderItemDefinition!(execute=@n10.AthenaGliderItemDefinition, Object=Get Array ElementSelectedAssets()) -> (then, CastFailed, AsAthena Glider Item Definition:Athena Glider Item Definition Object Reference) {is=title K2Node_DynamicCast}
   > n12 = LoadAsset_Blocking(Asset=GetOverrideSkelMeshPtr!(self:Athena Glider Item Definition Object Reference=@n11) -> (ReturnValue:Skeletal Mesh Soft Object Reference))
   > n13 = Cast To SkeletalMesh!(execute, Object=@n12) -> (then, CastFailed, AsSkeletal Mesh:Skeletal Mesh Object Reference) {is=title K2Node_DynamicCast}
   > n14 = Set GliderMesh_Ref(GliderMesh_Ref:Skeletal Mesh Object Reference=@n13) -> (Output_Get:Skeletal Mesh Object Reference)
   > Is Valid(InputObject=@n14)
   > n15 = GetOverrideAndVariantMaterials!(execute, self) -> (
    then
    variant_materials_list:Array of Material Instance Constant Object References
    override_material_list:Array of Material Instance Constant Object References
   )
   > n16 = CheckIfHasOverrideOrVariantAndSpawn!(
    execute
    self
    List of Materials:Array of Material Instance Constant Object References=@n15.override_material_list
    SkeletalMeshToSpawn:Skeletal Mesh Object Reference=Get GliderMesh_Ref() -> (GliderMesh_Ref:Skeletal Mesh Object Reference)
    SetNumber:int=Get Element_LoadAsset_SetNumber_Index() -> (Element_LoadAsset_SetNumber_Index:int=0)
   ) -> (then, HasOverride:bool, HasVariant:bool) {is=title K2Node_CallFunction}
   > Set variant_materials_list(variant_materials_list:Array of Material Instance Constant Object References=@n15) -> (Output_Get:Array of Material Instance Constant Object References)
   > n17 = Branch(Condition=@n16)
   CheckIfHasOverrideOrVariantAndSpawn(
    execute=[@n21.then, @n23.then, @n17.then]
    List of Materials=Get variant_materials_list()
    SkeletalMeshToSpawn=Get GliderMesh_Ref()
    SetNumber=Get Element_LoadAsset_SetNumber_Index() -> (Element_LoadAsset_SetNumber_Index=0)
   )
   n18 = Branch(execute=@n17.else, Condition=IsInPlayInEditor(self=Get LevelEditorSubsystem()))
   > n19 = GetPIEWorlds(bIncludeDedicatedServer=false)
   > n20 = SpawnActor Season Asset Actor(
    WorldContextObject="Get (a copy)"(Array:Array of World Object References=@n19) -> (Output:World Object Reference)
    Class=&SeasonAsset_Actor_C
    SpawnTransform_Location="0,0,2100"
    SpawnTransform_Rotation="0, 0, 0"
    SpawnTransform_Scale="1,1,1"
    CollisionHandlingOverride=Undefined
    TransformScaleMethod=MultiplyWithRoot
   )
   > SetSkeletalMeshAsset(self=Get SeasonAsset_SkeletalMesh(self=@n20), NewMesh=Get GliderMesh_Ref())
   > n21 = Add(TargetMap=Get AssetsSpawned_Test(), Key=@n20, Value=Get Element_LoadAsset_SetNumber_Index() -> (Element_LoadAsset_SetNumber_Index=0)) {is=title K2Node_CallFunction}
   n22 = SpawnActorFromObject(execute=@n18.else, ObjectToUse=Get GliderMesh_Ref(), Location_X=0, Location_Y=0, Location_Z=2100, bTransient=false)
   > n23 = Add(TargetMap=Get AssetsSpawned_Test(), Key=@n22, Value=Get Element_LoadAsset_SetNumber_Index() -> (Element_LoadAsset_SetNumber_Index=0))
   n24 = Cast To AthenaPickaxeItemDefinition!(execute=@n10.AthenaPickaxeItemDefinition, Object=Get Array ElementSelectedAssets()) -> (then, CastFailed, AsAthena Pickaxe Item Definition:Athena Pickaxe Item Definition Object Reference) {is=title K2Node_DynamicCast}
   n25 = Get WeaponDefinition!(self:Athena Pickaxe Item Definition Object Reference=@n24) -> (WeaponDefinition:Fort Weapon Melee Item Definition Object Reference)
   > n26 = Set PickaxeMesh_Ref(PickaxeMesh_Ref:Skeletal Mesh Object Reference=GetWeaponMeshOverride!(self:Fort Weapon Item Definition Object Reference=@n25) -> (ReturnValue:Skeletal Mesh Object Reference)) -> (Output_Get:Skeletal Mesh Object Reference)
   > Is Valid(InputObject=@n26)
   > n27 = CheckForDualWield!(execute, self, WeaponDefinition_Ref:Fort Weapon Melee Item Definition Object Reference=@n25) -> (then, OffHand_SkeletalMesh_Ref:Skeletal Mesh Object Reference) {is=title K2Node_CallFunction}
   > n28 = Is Valid(InputObject=@n27)
   n29 = GetOverrideAndVariantMaterials(execute=@n28.Is Not Valid)
   > n30 = CheckIfHasOverrideOrVariantAndSpawn(
    List of Materials=@n29.override_material_list
    SkeletalMeshToSpawn=Get PickaxeMesh_Ref()
    SetNumber=Get Element_LoadAsset_SetNumber_Index() -> (Element_LoadAsset_SetNumber_Index=0)
   )
   > Set variant_materials_list(variant_materials_list=@n29)
   > n31 = Branch(Condition=@n30)
   CheckIfHasOverrideOrVariantAndSpawn(
    execute=[@n37.then, @n35.then, @n31.then]
    List of Materials=Get variant_materials_list()
    SkeletalMeshToSpawn=Get PickaxeMesh_Ref()
    SetNumber=Get Element_LoadAsset_SetNumber_Index() -> (Element_LoadAsset_SetNumber_Index=0)
   )
   n32 = Branch(execute=@n31.else, Condition=IsInPlayInEditor(self=Get LevelEditorSubsystem()))
   > n33 = GetPIEWorlds(bIncludeDedicatedServer=false)
   > n34 = SpawnActor Season Asset Actor(
    WorldContextObject="Get (a copy)"(Array=@n33)
    Class=&SeasonAsset_Actor_C
    SpawnTransform_Location="0,0,2100"
    SpawnTransform_Rotation="0, 0, 0"
    SpawnTransform_Scale="1,1,1"
    CollisionHandlingOverride=Undefined
    TransformScaleMethod=MultiplyWithRoot
   )
   > SetSkeletalMeshAsset(self=Get SeasonAsset_SkeletalMesh(self=@n34), NewMesh=Get PickaxeMesh_Ref())
   > n35 = Add(TargetMap=Get AssetsSpawned_Test(), Key=@n34, Value=Get Element_LoadAsset_SetNumber_Index() -> (Element_LoadAsset_SetNumber_Index=0))
   n36 = SpawnActorFromObject(execute=@n32.else, ObjectToUse=Get PickaxeMesh_Ref(), Location_X=0, Location_Y=0, Location_Z=2100, bTransient=false)
   > n37 = Add(TargetMap=Get AssetsSpawned_Test(), Key=@n36, Value=Get Element_LoadAsset_SetNumber_Index() -> (Element_LoadAsset_SetNumber_Index=0))
   SpawnDualWield!(
    execute=@n28.Is Valid
    self
    OffHandMesh_Ref:Skeletal Mesh Object Reference=@n27
    SetNumber:int=Get Element_LoadAsset_SetNumber_Index() -> (Element_LoadAsset_SetNumber_Index=0)
   ) -> (then)
   n38 = Cast To AthenaBackpackItemDefinition!(execute=@n10.AthenaBackpackItemDefinition, Object=Get Array ElementSelectedAssets()) -> (then, CastFailed, AsAthena Backpack Item Definition:Athena Backpack Item Definition Object Reference) {is=title K2Node_DynamicCast}
   n39 = GetCharacterParts!(self:Athena Character Part Item Definition Object Reference=@n38) -> (ReturnValue:Array of Custom Character Part Object References)
   > n40 = Set BackPackSkelMesh_Ref(BackPackSkelMesh_Ref:Skeletal Mesh Object Reference=GetSkeletalMesh!(self:Custom Character Part Object Reference="Get (a copy)"(Array:Array of Custom Character Part Object References=@n39) -> (Output:Custom Character Part Object Reference)) -> (ReturnValue:Skeletal Mesh Object Reference)) -> (Output_Get:Skeletal Mesh Object Reference)
   > Is Valid(InputObject=@n40)
   > n41 = GetOverrideAndVariantMaterials()
   > n42 = CheckIfHasOverrideOrVariantAndSpawn(
    List of Materials=@n41.override_material_list
    SkeletalMeshToSpawn=Get BackPackSkelMesh_Ref() -> (BackPackSkelMesh_Ref:Skeletal Mesh Object Reference)
    SetNumber=Get Element_LoadAsset_SetNumber_Index() -> (Element_LoadAsset_SetNumber_Index=0)
   )
   > Set variant_materials_list(variant_materials_list=@n41)
   > n43 = Branch(Condition=@n42)
   CheckIfHasOverrideOrVariantAndSpawn(
    execute=[@n49.then, @n47.then, @n43.then]
    List of Materials=Get variant_materials_list()
    SkeletalMeshToSpawn=Get BackPackSkelMesh_Ref()
    SetNumber=Get Element_LoadAsset_SetNumber_Index() -> (Element_LoadAsset_SetNumber_Index=0)
   )
   n44 = Branch(execute=@n43.else, Condition=IsInPlayInEditor(self=Get LevelEditorSubsystem()))
   > n45 = GetPIEWorlds(bIncludeDedicatedServer=false)
   > n46 = SpawnActor Season Asset Actor(
    WorldContextObject="Get (a copy)"(Array:Array of World Object References=@n45) -> (Output:World Object Reference)
    Class=&SeasonAsset_Actor_C
    SpawnTransform_Location="0,0,2100"
    SpawnTransform_Rotation="0, 0, 0"
    SpawnTransform_Scale="1,1,1"
    CollisionHandlingOverride=Undefined
    TransformScaleMethod=MultiplyWithRoot
   )
   > SetSkeletalMeshAsset(self=Get SeasonAsset_SkeletalMesh(self=@n46), NewMesh=Get BackPackSkelMesh_Ref())
   > n47 = Add(TargetMap=Get AssetsSpawned_Test(), Key=@n46, Value=Get Element_LoadAsset_SetNumber_Index() -> (Element_LoadAsset_SetNumber_Index=0))
   n48 = SpawnActorFromObject(execute=@n44.else, ObjectToUse=Get BackPackSkelMesh_Ref(), Location_X=0, Location_Y=0, Location_Z=2100, bTransient=false)
   > n49 = Add(TargetMap=Get AssetsSpawned_Test(), Key=@n48, Value=Get Element_LoadAsset_SetNumber_Index() -> (Element_LoadAsset_SetNumber_Index=0))
   "On Clicked (btn_DeleteSeasonAssets)"() {is=title K2Node_ComponentBoundEvent}
   > CollectGarbage()
   > DeleteCurrentAssetsSpawned()
   n50 = "On Check State Changed (SpawnGliders_CheckBox)"!() -> (OutputDelegate, then, bIsChecked:bool) {is=title K2Node_ComponentBoundEvent}
   > Set SpawnGliders Is Checked(SpawnGliders Is Checked:bool=@n50.bIsChecked) -> (Output_Get:bool)
   n51 = "OnTextChanged (Editable Text) (TimeOfDayInput)"!() -> (OutputDelegate, then, Text:"Text (by ref)") {is=title K2Node_ComponentBoundEvent}
   > Set version_name(version_name:String="To String (Text)"(InText=@n51.Text)) -> (Output_Get:String)
   "On Clicked (btn_SpawnSelectedAssets)"() {is=title K2Node_ComponentBoundEvent}
   > CollectGarbage()
   > ClearAssetNameArray()
   > n52 = SpawnContentBrowser()
   "On Clicked (btn_HideSet)"() {is=title K2Node_ComponentBoundEvent}
   > RemoveSetFilter()
   "On Clicked (btn_Debug)"() {is=title K2Node_ComponentBoundEvent}
   > n53 = GetPIEWorlds(bIncludeDedicatedServer=false)
   > n54 = SpawnActor Season Asset Actor(
    WorldContextObject="Get (a copy)"(Array=@n53)
    Class=&SeasonAsset_Actor_C
    SpawnTransform_Location="0,0,2100"
    SpawnTransform_Rotation="0, 0, 0"
    SpawnTransform_Scale="1,1,1"
    CollisionHandlingOverride=Undefined
    TransformScaleMethod=MultiplyWithRoot
   )
   > SetSkeletalMeshAsset(
    self=Get SeasonAsset_SkeletalMesh(self=@n54)
    NewMesh=/BRCosmetics/Gadgets/Assets/VinderTech_GliderChute/Glider_AuraCop/Meshes/AuraCop_Glider
   )
   "On Clicked (btn_RenderAssets)"() {is=title K2Node_ComponentBoundEvent}
   > Clear(TargetMap:Map of Actor Object References to Booleans=Get MapAssetsStatus() -> (MapAssetsStatus:Map of Actor Object References to Booleans)) {is=title K2Node_CallFunction}
   > n55 = GetEditorWorld(self=Get UnrealEditorSubsystem())
   > n56 = GetAllActorsOfClass!(execute, WorldContextObject:Object Reference=@n55, ActorClass:Actor Class Reference=&Actor) -> (then, OutActors:Array of Actor Object References)
   > n57 = For Each Loop(Array:Array of Actor Object References=@n56) -> (Array Element:Actor Object Reference)
   n58 = GetEditorWorld(execute=@n57.Completed, self=Get UnrealEditorSubsystem())
   > n59 = GetActorOfClass!(
    execute
    WorldContextObject:Object Reference=@n58
    ActorClass:Actor Class Reference=/Game/ArtTools/HS_Season_Tool/RenderSeasonAssets/BP_RenderSeasonAssets.BP_RenderSeasonAssets_C
   ) -> (then, ReturnValue:BP Render Season Assets Object Reference)
   > n60 = GetActorOfClass(WorldContextObject=@n58, ActorClass=/Game/TimeOfDay/TODM_Sequence/Blueprint/BP_Asteria_DaySequence.BP_Asteria_DaySequence_C) -> (ReturnValue:BP Asteria Day Sequence Object Reference)
   > SetIsTemporarilyHiddenInEditor!(execute, self:Actor Object Reference=[@n60, @n59], bIsHidden:bool=false) -> (then)
   > n61 = Execute Python Script!(
    execute
    PythonScript="import unreal\r\n\r\nall_actors = unreal.EditorLevelLibrary.get_all_level_actors()\r\n\r\nrender_light_actors = [actor for actor in all_actors if actor.get_folder_path() == \"RenderLight\"]"
   ) -> (then, ReturnValue, render_light_actors:Array of Actor Object References) {outputs [render_light_actors]}
   > n62 = For Each Loop(Array=@n61.render_light_actors)
   n63 = Keys(execute=@n62.Completed, TargetMap:Map of Actor Object References to Integers=Get AssetsSpawned_Test()) -> (Keys:Array of Actor Object References)
   > n64 = For Each Loop with Delay!(Exec:Exec, Array:Array of Actor Object References=@n63, Delay:double=2.5) -> (Loop Body:Exec, Array Element:Actor Object Reference, Array Index:int, Completed:Exec) {is=title K2Node_MacroInstance}
   n65 = For Each Loop(
    Exec=@n64.Completed
    Array:Array of Texture 2D Object References=Get ArrayOfRenderTextures() -> (ArrayOfRenderTextures:Array of Texture 2D Object References)
   ) -> (Array Element:Texture 2D Object Reference)
   n66 = Get MapAssetsStatus()
   n67 = Keys(execute=@n65.Completed, TargetMap:Map of Actor Object References to Booleans=@n66)
   > n68 = For Each Loop(Array:Array of Actor Object References=@n67) -> (Array Element:Actor Object Reference)
   > SetIsTemporarilyHiddenInEditor(
    self=@n68
    bIsHidden=Find(TargetMap:Map of Actor Object References to Booleans=@n66, Key:"Actor Object Reference (by ref)"=@n68) -> (Value:bool)
   )
   Set AdjustMinAlpha!(execute=@n65.LoopBody, AdjustMinAlpha:float=1, self:Texture Object Reference=@n65) -> (then, Output_Get:float)
   > Set AdjustMaxAlpha!(execute, AdjustMaxAlpha:float=0, self:Texture Object Reference=@n65) -> (then, Output_Get:float)
   n69 = Set ElementAssetSpawned(execute=@n64.Loop Body, ElementAssetSpawned:Actor Object Reference=@n64) -> (Output_Get:Actor Object Reference)
   > SetIsTemporarilyHiddenInEditor(self=@n69, bIsHidden=false)
   > Set LocationBackup(LocationBackup:Vector=Get Actor Location!(self=@n69) -> (ReturnValue:Vector)) -> (Output_Get:Vector)
   > Set Actor Location And Rotation!(
    execute
    self=Get ElementAssetSpawned() -> (ElementAssetSpawned:Actor Object Reference)
    NewLocation:Vector=Get World Location!(self:Scene Component Object Reference=Get SKM_RenderSeasonAsset!(self:BP Render Season Assets Object Reference=@n59) -> (SKM_RenderSeasonAsset:Skeletal Mesh Component Object Reference)) -> (ReturnValue:Vector) {is=title K2Node_CallFunction}
    NewRotation="0,-90,90"
    bSweep=false
    bTeleport=true
   ) -> (then, SweepHitResult, ReturnValue)
   n70 = Get ElementAssetSpawned()
   > DelayUntilNextTick!(execute, WorldContextObject:Object Reference=@n70) -> (then)
   > DelayUntilNextTick(WorldContextObject=@n70)
   > n71 = Render Target 2D Create Static Texture 2D Editor Only!(
    execute
    RenderTarget:Texture Render Target 2D Object Reference=/Game/ArtTools/HS_Season_Tool/RenderSeasonAssets/RT_Icon_SeasonAssets
    Name:String=Append(A=render_, B=GetDisplayName(Object=@n70))
    CompressionSettings:TextureCompressionSettings Enum=TC_Default
    MipSettings:TextureMipGenSettings Enum=TMGS_FromTextureGroup
   ) -> (then, ReturnValue:Texture 2D Object Reference)
   > Add(TargetArray:Array of Texture 2D Object References=Get ArrayOfRenderTextures(), NewItem:"Texture 2D Object Reference (by ref)"=@n71) {is=title K2Node_CallArrayFunction}
   > DelayUntilNextTick(WorldContextObject=Get ElementAssetSpawned())
   > Set Actor Location And Rotation(
    self=Get ElementAssetSpawned()
    NewLocation=Get LocationBackup() -> (LocationBackup:Vector)
    NewRotation="0, 0, 0"
    bSweep=false
    bTeleport=true
   )
   SetIsTemporarilyHiddenInEditor(execute=@n62.LoopBody, self=@n62, bIsHidden=false)
   Add(
    execute=@n57.LoopBody
    TargetMap:Map of Actor Object References to Booleans=Get MapAssetsStatus()
    Key=@n57
    Value:"Boolean (by ref)"=IsTemporarilyHiddenInEditor!(self:Actor Object Reference=@n57, bIncludeParent:bool=false) -> (ReturnValue:bool)
   ) {is=title K2Node_CallFunction}
   > SetIsTemporarilyHiddenInEditor(self=@n57, bIsHidden=true)
   n72 = "OnTextChanged (Editable Text) (SetNameInput)"!() -> (OutputDelegate, then, Text:"Text (by ref)") {is=title K2Node_ComponentBoundEvent}
   > Set SetName(SetName:Text=@n72.Text) -> (Output_Get:Text)
   "On Clicked (btn_KeepSet)"() {is=title K2Node_ComponentBoundEvent}
   > KeepSetFilter()
   "On Clicked (btn_ShowAll)"() {is=title K2Node_ComponentBoundEvent}
   > n73 = Keys(TargetMap:Map of Actor Object References to Integers=Get AssetsSpawned_Test())
   > n74 = For Each Loop(Array=@n73)
   > SetIsTemporarilyHiddenInEditor(self=@n74, bIsHidden=false)
   "On Clicked (btn_DisplaySelectedAssets)"() {is=title K2Node_ComponentBoundEvent}
   > DisplaySelectedAssets() {is=title K2Node_CallFunction}
   n75 = "On Check State Changed (SpawnPickaxes_CheckBox)"!() -> (OutputDelegate, then, bIsChecked:bool) {is=title K2Node_ComponentBoundEvent}
   > Set SpawnPickaxes Is Checked(SpawnPickaxes Is Checked:bool=@n75.bIsChecked) -> (Output_Get:bool)
   n76 = "On Check State Changed (SpawnBackpacks_CheckBox)"!() -> (OutputDelegate, then, bIsChecked:bool) {is=title K2Node_ComponentBoundEvent}
   > Set SpawnBackpacks Is Checked(SpawnBackpacks Is Checked:bool=@n76.bIsChecked) -> (Output_Get:bool)
   n77 = RotateSun!() -> (OutputDelegate, then, RotateTo:double=~) {is=title K2Node_CustomEvent}
   > n78 = Cast To FortDaySequenceActor!(
    execute
    Object:Fort Time Of Day Manager Interface Interface=GetGlobalTimeOfDayManager!(WorldContextObject:Object Reference) -> (ReturnValue:Fort Time Of Day Manager Interface Interface)
   ) -> (then, CastFailed, AsFort Day Sequence Actor:Fort Day Sequence Actor Object Reference) {is=title K2Node_DynamicCast}
   > Set InitialSunRotation(InitialSunRotation:float=BreakRotator(InRot=Get Actor Rotation!(self:Actor Object Reference=@n78) -> (ReturnValue:Rotator) {is=title K2Node_CallFunction}).Yaw) -> (Output_Get:float)
   n79 = BreakRotator(InRot=Get Actor Rotation(self=@n78))
   n80 = Get InitialSunRotation() -> (InitialSunRotation:float)
   > Set Actor Rotation!(
    execute
    self:Actor Object Reference=@n78
    NewRotation:Rotator=MakeRotator(Roll=@n79, Pitch=@n79.Pitch, Yaw=float + float(A=@n77.RotateTo, B=@n80))
    bTeleportPhysics:bool=false
   ) -> (then, ReturnValue:bool) {is=title K2Node_CallFunction}
   Event Construct()
   > LoadWorldPartition()
   n81 = For Each Loop(
    Exec=@n1.then
    Array:Array of Custom Character Part Soft Object References=Get BaseCharacterParts!(self:Athena Character Item Definition Object Reference=@n1) -> (BaseCharacterParts:Array of Custom Character Part Soft Object References)
   ) -> (Array Element:Custom Character Part Soft Object Reference)
   > n82 = LoadAsset_Blocking(Asset=@n81)
   > n83 = Cast To CustomCharacterPart!(execute, Object=@n82) -> (then, CastFailed, AsCustom Character Part:Custom Character Part Object Reference) {is=title K2Node_DynamicCast}
   > Add(
    TargetArray:Array of Skeletal Mesh Object References=Get CharacterSkeletalMesh()
    NewItem:"Skeletal Mesh Object Reference (by ref)"=GetSkeletalMesh(self=@n83)
   ) {is=title K2Node_CallArrayFunction}
   n84 = SpawnActorFromClass!(
    execute=@n81.Completed
    ActorClass:Actor Class Reference=&/Character_SeasonAsset_Actor.Character_SeasonAsset_Actor_C
    Location:Vector="0,0,2100"
    Rotation:Rotator
    bTransient:bool
   ) -> (then, ReturnValue:Character Season Asset Actor Object Reference)
   > n85 = For Each Loop(Array:Array of Skeletal Mesh Object References=Get CharacterSkeletalMesh()) -> (Array Element:Skeletal Mesh Object Reference)
   SetSkeletalMeshAsset(
    execute=@n87.0
    self=Get Character_Body_SkeletalMesh!(self:Character Season Asset Actor Object Reference=@n84) -> (Character_Body_SkeletalMesh:Skeletal Mesh Component Object Reference)
    NewMesh=@n85
   )
   SetSkeletalMeshAsset(
    execute=@n87.2
    self=Get Character_Head_SkeletalMesh!(self:Character Season Asset Actor Object Reference=@n84) -> (Character_Head_SkeletalMesh:Skeletal Mesh Component Object Reference)
    NewMesh=@n85
   )
   SetSkeletalMeshAsset(
    execute=@n87.1
    self=Get Character_Face_SkeletalMesh!(self:Character Season Asset Actor Object Reference=@n84) -> (Character_Face_SkeletalMesh:Skeletal Mesh Component Object Reference)
    NewMesh=@n85
   )
   IsInPlayInEditor(self=Get LevelEditorSubsystem())
   AssetsToFolder(execute=@n86.then)
   Event PreConstruct()
   n86 = Add(
    execute=@n85.Completed
    TargetMap:Map of Actor Object References to Integers=Get AssetsSpawned_Test()
    Key=@n84
    Value:"Integer (by ref)"=Get Element_LoadAsset_SetNumber_Index() -> (Element_LoadAsset_SetNumber_Index=0)
   ) {is=title K2Node_CallFunction}
   n87 = Switch on Int!(execute:Exec=@n85.LoopBody, Selection:int=@n85.Array Index) -> (Default:Exec, 0:Exec, 1:Exec, 2:Exec)
  ]}
 GetAssetsPathAndLoad {nodes [
   n1 = GetAssetsPathAndLoad!() -> (then, Asset_List:Array of Strings) {is=title K2Node_FunctionEntry}
   > n2 = Set cosmetic_assets_list(cosmetic_assets_list:Array of Strings=@n1) -> (Output_Get:Array of Strings)
   > n3 = For Each Loop(Array:Array of Strings=@n2) -> (Array Element:String)
   SortStringArray!(execute=@n3.Completed, TargetArray:Array of Strings=Get SetNames(), bStableSort:bool=true, SortOrder:EArraySortOrder Enum=Ascending) -> (then)
   > n4 = For Each Loop(Array=Get SetNames())
   Return Node!(execute=@n4.Completed)
   > n5 = For Each Loop(Array=Get cosmetic_assets_list() -> (cosmetic_assets_list:Array of Strings))
   > Branch(Condition=Contains(SearchIn=@n5, Substring=@n4))
   > n6 = Branch(Condition=IsInPlayInEditor(self=Get LevelEditorSubsystem()))
   > n7 = GetAssetByObjectPath!(execute, self:Asset Registry Interface=GetAssetRegistry(), ObjectPath:Name=String To Name(InString=@n5), bIncludeOnlyOnDiskAssets:bool=false) -> (then, ReturnValue:Asset Data Structure)
   > Add(
    TargetMap:Map of Object References to Integers=Get ArrayOfLoadedAssets_Test()
    Key:"Object Reference (by ref)"=GetAsset(InAssetData=@n7)
    Value=@n4.Array Index
   )
   n8 = LoadAsset(execute=@n6.else, AssetPath=@n5)
   > Add(TargetMap=Get ArrayOfLoadedAssets_Test(), Key=@n8, Value=@n4.Array Index)
   n9 = Set ElementOfCosmeticAssetsList(execute=@n3.LoopBody, ElementOfCosmeticAssetsList:String=@n3) -> (Output_Get:String)
   n10 = RightChop!(SourceString:String=@n9, Count:int=int + int(A=FindSubstring(SearchIn=@n9, Substring=., bSearchFromEnd=true))) -> (ReturnValue:String)
   n11 = FindSubstring(SearchIn=@n10, Substring=_)
   > n12 = Branch(Condition="Equal (Integer)"(A=@n11))
   > Add Unique(TargetArray=Get SetNames(), NewItem=@n10)
   Add Unique(execute=@n12.else, TargetArray=Get SetNames(), NewItem=RightChop(SourceString=@n10, Count=int + int(A=@n11)))
  ]}
 GetLocationActors {nodes [
   GetLocationActors() {is=title K2Node_FunctionEntry}
   n1 = Get LocationActors()
   > n2 = Branch(Condition="integer <= integer"(A=Length(TargetArray:Array of Actor Object References=@n1), B=16))
   > n3 = Branch(Condition=IsInPlayInEditor(self=Get LevelEditorSubsystem()))
   > n4 = GetPIEWorlds(bIncludeDedicatedServer=false)
   > n5 = GetAllActorsOfClass(WorldContextObject="Get (a copy)"(Array=@n4), ActorClass=&/Location_SeasonAsset.Location_SeasonAsset_C) -> (OutActors:Array of Location Season Asset Object References)
   > n6 = SortActorArray!(execute, InActors:Array of Actor Object References=@n5) -> (then, OutSortedActors:Array of Actor Object References)
   > Set LocationActors(LocationActors:Array of Actor Object References=@n6) -> (Output_Get:Array of Actor Object References)
   n7 = GetEditorWorld(execute=@n3.else, self=Get UnrealEditorSubsystem())
   > n8 = GetAllActorsOfClass(WorldContextObject=@n7, ActorClass=&Location_SeasonAsset_C)
   > n9 = SortActorArray(InActors=@n8)
   > Set LocationActors(LocationActors=@n9)
   n10 = Reverse for Each Loop!(Exec:Exec=@n2.else, Array:Array of Actor Object References=@n1) -> (LoopBody:Exec, ArrayIndex:int, ArrayElement:Actor Object Reference, Completed:Exec) {is=title K2Node_MacroInstance}
   > n11 = Is Valid(InputObject=@n10.ArrayElement)
   Remove Index!(execute=@n11.Is Not Valid, TargetArray:Array of Actor Object References=Get LocationActors(), IndexToRemove:int=@n10) -> (then)
  ]}
 GetOverrideAndVariantMaterials {nodes [
   GetOverrideAndVariantMaterials() {is=title K2Node_FunctionEntry}
   > n1 = Execute Python Script!(
    execute
    PythonScript=Get GetVariant_PythonScript() -> (GetVariant_PythonScript:String)
    asset_name:String=Get CurrentAsset_Name() -> (CurrentAsset_Name:String)
   ) -> (
    then
    ReturnValue
    override_material_list:Array of Material Instance Constant Object References
    variant_materials_list:Array of Material Instance Constant Object References
   ) {inputs [asset_name], outputs [override_material_list, variant_materials_list]}
   > n2 = Set override_material_list(override_material_list:Array of Material Instance Constant Object References=@n1.override_material_list) -> (Output_Get:Array of Material Instance Constant Object References)
   > n3 = Set variant_materials_list(variant_materials_list=@n1.variant_materials_list)
   > Return Node!(
    execute
    variant_materials_list:Array of Material Instance Constant Object References=@n3
    override_material_list:Array of Material Instance Constant Object References=@n2
   )
   Execute Python Script!(
    execute
    PythonScript="## Copyright Epic Games, Inc. All Rights Reserved.\r\nimport unreal\r\n\r\ndef get_asset_materials(asset_type, asset_name):\r\n    unreal.log(f\"Processing Asset: {asset_name} (Type: {asset_type})\")\r\n    ##Organizes the asset path given the asset type and the asset name\r\n    asset_path = \"/BRCosmetics/Athena/Items/Cosmetics/\" + asset_type + \"/\" + asset_name + \".\" + asset_name\r\n\r\n    \r\n    ##Loads the asset given the path\r\n    data_asset = unreal.EditorAssetLibrary.load_asset(asset_path)\r\n\r\n    ##Gets the Material Overrides\r\n    material_overrides_length = len(data_asset.get_editor_property(\"MaterialOverrides\"))\r\n    \r\n    if True:\r\n        \r\n        try:\r\n            ##Collects the ItemVariant property from editor\r\n            properties = data_asset.get_editor_property(\"ItemVariants\")\r\n            \r\n            ##Gathers the Array of ParticleOptions\r\n            part_options = properties[0].get_editor_property(\"PartOptions\")\r\n            if not part_options:\r\n                unreal.log(f\"{asset_name} has no part_options, check for Materil options\")\r\n                part_options = properties[0].get_editor_property(\"MaterialOptions\")\r\n            \r\n            ##Establishes the lenght so it can run through multiple variants (even if there's only one)\r\n            length = len(part_options)\r\n\r\n            for i in range(1, length):\r\n                ##Gets the length of the VariantParts \r\n                part_options_len = len(part_options[i].get_editor_property(\"VariantParts\"))\r\n                if(part_options_len == 1):\r\n                    variant_parts = part_options[i].get_editor_property(\"VariantParts\")[0]\r\n                    variant_parts_list.append(variant_parts)\r\n                    unreal.log(f\"Found material variant: {material_override.get_name()}\")\r\n                elif(part_options_len == 2):\r\n                    variant_parts = part_options[i].get_editor_property(\"VariantParts\")[0]\r\n                    variant_parts_list.append(variant_parts)\r\n                    variant_parts = part_options[i].get_editor_property(\"VariantParts\")[1]\r\n                    variant_parts_list.append(variant_parts)\r\n                    ##variant_parts_list.append(override_material_list)\r\n             \r\n        except:\r\n            print(\"The Asset: \" + asset_name + \" does not have any Variant Materials or Override Materials\")\r\n\r\n    if material_overrides_length > 0:\r\n        ##Collects the property MaterialOverrides\r\n        for material_override in data_asset.get_editor_property(\"MaterialOverrides\"):\r\n            override_material = material_override.get_editor_property(\"OverrideMaterial\")\r\n            override_material_list.append(override_material)\r\n            unreal.log(f\"Found material override: {override_material.get_name()}\")\r\n\r\nvariant_parts_list = []\r\noverride_material_list = []\r\n##\"Pickaxe_PointSmoke\" \"Pickaxe_TruckScale\" \"Pickaxe_BlowWire\" \"Pickaxe_RezzPickaxe\" \"Backpack_AgileRug\" \"Glider_MedievalSheath\"\r\nasset_name = asset_name\r\nif(\"Character\" in asset_name):\r\n    get_asset_materials(\"Characters\", asset_name)\r\nelse:\r\n    print(\"Invalid Asset Name\")"
    asset_name=Get CurrentAsset_Name()
   ) -> (then, ReturnValue, variant_parts_list:Wildcard) {inputs [asset_name], outputs [variant_parts_list]}
  ]}
 KeepSetFilter {nodes [
   KeepSetFilter() {is=title K2Node_FunctionEntry}
   > n1 = Keys(TargetMap=Get AssetsSpawned_Test())
   > n2 = For Each Loop(Array:Array of Actor Object References=@n1) -> (Array Element:Actor Object Reference)
   n3 = Get SetName() -> (SetName:Text)
   > n4 = Branch(Condition=Contains(SearchIn=GetDisplayName(Object=@n2), Substring="To String (Text)"(InText=@n3)))
   SetIsTemporarilyHiddenInEditor(execute=@n4.else, self=@n2, bIsHidden=true)
  ]}
 LoadWorldPartition {nodes [
   LoadWorldPartition() {is=title K2Node_FunctionEntry}
   > n1 = Branch(Condition=IsInPlayInEditor(self=Get LevelEditorSubsystem()))
   n2 = GetEditorWorld(execute=@n1.else, self=Get UnrealEditorSubsystem())
   > n3 = GetActorOfClass(WorldContextObject=@n2, ActorClass=/Script/Engine.LocationVolume) -> (ReturnValue:Location Volume Object Reference)
   > Load(self:Location Volume Object Reference=@n3)
  ]}
 RemoveSetFilter {nodes [
   RemoveSetFilter() {is=title K2Node_FunctionEntry}
   > n1 = Keys(TargetMap=Get AssetsSpawned_Test())
   > n2 = For Each Loop(Array=@n1)
   n3 = Get SetName()
   > Branch(Condition=Contains(SearchIn=GetDisplayName(Object=@n2), Substring="To String (Text)"(InText=@n3)))
   > SetIsTemporarilyHiddenInEditor(self=@n2, bIsHidden=true)
  ]}
 ShowTab {nodes [
   n1 = ShowTab!() -> (then, VerticalBox:Vertical Box Object Reference) {is=title K2Node_FunctionEntry}
   > n2 = For Each Loop(Array:Array of Vertical Box Object References=Get PanelsArray() -> (PanelsArray:Array of Vertical Box Object References)) -> (Array Element:Vertical Box Object Reference)
   SetVisibility(execute=@n2.Completed, self=@n1, InVisibility=Visible)
   Branch(
    execute=@n2.LoopBody
    Condition="Not Equal (Enum)"!(
     A:ESlateVisibility Enum=GetVisibility!(self:Widget Object Reference=@n2) -> (ReturnValue:ESlateVisibility Enum=Visible)
     B:ESlateVisibility Enum=Hidden
    ) -> (ReturnValue:bool)
   )
   > SetVisibility(self=@n2, InVisibility=Hidden)
  ]}
 SpawnContentBrowser {nodes [
   SpawnContentBrowser() {is=title K2Node_FunctionEntry}
   > n1 = GetSelectedAssets()
   > n2 = For Each Loop(Array:Array of Object References=@n1) -> (Array Element:Object Reference)
   > Add(TargetArray:Array of Strings=Get Asset_List(), NewItem:"String (by ref)"=Get Object Path String(Object=@n2)) {is=title K2Node_CallArrayFunction}
  ]}
 SpawnDualWield {nodes [
   n1 = SpawnDualWield!() -> (then, OffHandMesh_Ref:Skeletal Mesh Object Reference, SetNumber:int) {is=title K2Node_FunctionEntry}
   n2 = Get DualWieldArray() -> (DualWieldArray:Array of Skeletal Mesh Object References)
   > Add(TargetArray:Array of Skeletal Mesh Object References=@n2, NewItem:"Skeletal Mesh Object Reference (by ref)"=@n1)
   > Add(TargetArray=@n2, NewItem=Get PickaxeMesh_Ref())
   > n3 = For Each Loop(Array:Array of Skeletal Mesh Object References=@n2) -> (Array Element:Skeletal Mesh Object Reference)
   Clear(execute=@n3.Completed, TargetArray=Get DualWieldArray()) {is=title K2Node_CallArrayFunction}
   > AssetsToFolder()
   n4 = Branch(execute=@n3.LoopBody, Condition=IsInPlayInEditor(self=Get LevelEditorSubsystem()))
   > n5 = GetPIEWorlds(bIncludeDedicatedServer=false)
   > n6 = SpawnActor Season Asset Actor(
    WorldContextObject="Get (a copy)"(Array=@n5)
    Class=&SeasonAsset_Actor_C
    SpawnTransform_Location="0,0,2100"
    SpawnTransform_Rotation="0, 0, 0"
    SpawnTransform_Scale="1,1,1"
    CollisionHandlingOverride=Undefined
    TransformScaleMethod=MultiplyWithRoot
   )
   > SetSkeletalMeshAsset(self=Get SeasonAsset_SkeletalMesh(self=@n6), NewMesh=@n3)
   > Add(
    TargetMap:Map of Actor Object References to Integers=Get AssetsSpawned_Test()
    Key:"Actor Object Reference (by ref)"=@n6
    Value=@n1.SetNumber
   ) {is=title K2Node_CallFunction}
   n7 = SpawnActorFromObject(execute=@n4.else, ObjectToUse=@n3, Location_X=0, Location_Y=0, Location_Z=2100, bTransient=false)
   > Add(TargetMap=Get AssetsSpawned_Test(), Key=@n7, Value=@n1.SetNumber)
  ]}
 SpawnNewLocation {nodes [
   SpawnNewLocation() {is=title K2Node_FunctionEntry}
   > n1 = Branch(Condition=IsInPlayInEditor(self=Get LevelEditorSubsystem()))
   > n2 = GetPIEWorlds(bIncludeDedicatedServer=false)
   n3 = Get LocationActors()
   n4 = Get Actor Location!(self="Get (a copy)"(Array:Array of Actor Object References=@n3, Dimension 1=Last Index!(TargetArray:Array of Actor Object References=@n3) -> (ReturnValue=0)) -> (Output:Actor Object Reference)) -> (ReturnValue_X:double, ReturnValue_Y:double, ReturnValue_Z:double)
   > SpawnActor Season Asset Actor!(
    execute
    WorldContextObject="Get (a copy)"(Array:Array of World Object References=@n2) -> (Output:World Object Reference)
    Class=&SeasonAsset_Actor_C
    SpawnTransform_Location_X:double=float - float(A=@n4, B=310)
    SpawnTransform_Location_Y:double=@n4.ReturnValue_Y
    SpawnTransform_Location_Z:double=@n4.ReturnValue_Z
    SpawnTransform_Rotation="0, 0, 0"
    SpawnTransform_Scale="1,1,1"
    CollisionHandlingOverride=Undefined
    TransformScaleMethod=MultiplyWithRoot
    Owner
    Instigator
   ) -> (then, ReturnValue)
   n5 = Get LocationActors()
   n6 = Get Actor Location(self="Get (a copy)"(Array:Array of Actor Object References=@n5, Dimension 1=Last Index(TargetArray=@n5) -> (ReturnValue=0)) -> (Output:Actor Object Reference))
   n7 = SpawnActorFromClass!(
    execute=@n1.else
    ActorClass=&Location_SeasonAsset_C
    Location_X:double=float - float(A=@n6, B=310)
    Location_Y:double=@n6.ReturnValue_Y
    Location_Z:double=@n6.ReturnValue_Z
    Rotation
    bTransient
   ) -> (then, ReturnValue:Location Season Asset Object Reference)
   > Add(
    TargetArray:Array of Actor Object References=Get NewLocationActors() -> (NewLocationActors:Array of Actor Object References)
    NewItem:"Actor Object Reference (by ref)"=@n7
   ) {is=title K2Node_CallArrayFunction}
   > n8 = For Each Loop(Array:Array of Actor Object References=Get NewLocationActors()) -> (Array Element:Actor Object Reference)
   n9 = Get LocationActors()
   Execute Python Script!(
    execute=@n8.Completed
    PythonScript="import unreal\r\n\r\nactors = actors\r\nfolder_name = \"Location_SeasonAsset\"\r\nfolder_path =  folder_name\r\n\r\nfor actor in actors:\r\n    if isinstance(actor, unreal.Actor):\r\n        actor.set_folder_path(folder_path)"
    actors:Array of Actor Object References=@n9
   ) -> (then, ReturnValue) {inputs [actors]}
   > Clear(TargetArray:Array of Actor Object References=Get NewLocationActors())
   > AssetsToLocationSpawn()
   Add(execute=@n8.LoopBody, TargetArray=@n9, NewItem=@n8)
  ]}
 SpawnSeasonAsset {nodes [
   SpawnSeasonAsset() {is=title K2Node_FunctionEntry}
   > n1 = Execute Python Script!(
    execute
    PythonScript="## Copyright Epic Games, Inc. All Rights Reserved.\r\nimport unreal\r\n\r\ndef get_release_assets(asset_type, version_name):\r\n    cosmetics_path = '&Cosmetics/' + asset_type\r\n    asset_reg = unreal.AssetRegistryHelpers.get_asset_registry()\r\n    assets = asset_reg.get_assets_by_path(cosmetics_path)\r\n    for asset in assets:\r\n        asset_path = asset.get_full_name().split(' ')[-1]\r\n        data_asset = unreal.EditorAssetLibrary.load_asset(asset_path)\r\n        try:\r\n            version_name_aux = str(data_asset.get_editor_property(\"FortReleaseVersion\").get_editor_property(\"VersionName\"))\r\n            if version_name_aux == version_name:\r\n                cosmetic_assets_list.append(asset_path)\r\n        except:\r\n            continue\r\n\r\ncosmetic_assets_list = []\r\nversion_name = version_name\r\nasset_type = \"\"\r\nbackpack = backpack\r\nglider = glider\r\npickaxe = pickaxe\r\nerror_log = \"\"\r\n\r\nif not \" \" in version_name:\r\n    if backpack:\r\n        get_release_assets(\"Backpacks\", version_name)\r\n    if glider:\r\n        get_release_assets(\"Gliders\", version_name)\r\n    if pickaxe:\r\n        get_release_assets(\"Pickaxes\", version_name)\r\nelse:\r\n    error_log = \"Invalid Version - Remove Space\""
    version_name:String=Get version_name() -> (version_name:String)
    backpack:bool=Get SpawnBackpacks Is Checked() -> (SpawnBackpacks Is Checked:bool)
    glider:bool=Get SpawnGliders Is Checked() -> (SpawnGliders Is Checked:bool)
    pickaxe:bool=Get SpawnPickaxes Is Checked() -> (SpawnPickaxes Is Checked:bool)
   ) -> (then, ReturnValue, cosmetic_assets_list:Array of Strings) {inputs [version_name,backpack,glider,pickaxe], outputs [cosmetic_assets_list]}
   > Set Asset_List(Asset_List:Array of Strings=@n1.cosmetic_assets_list) -> (Output_Get:Array of Strings)
  ]}
]
variables [
 Array ElementSelectedAssets {type=Object, default=null}
 ArrayOfLoadedAssets {type="array<Object>", default []}
 ArrayOfLoadedAssets_Test {default {}}
 ArrayOfRenderTextures {type="array<Texture2D>", default []}
 Asset_List {type="array<string>", default []}
 AssetName {type="array<string>", default []}
 AssetsSpawned {type="array<Actor>"}
 AssetsSpawned_Test {default {}}
 AssetType {type=string, default=""}
 BackPackSkelMesh_Ref {type=SkeletalMesh, default=null}
 CharacterSkeletalMesh {type="array<SkeletalMesh>", default []}
 ContentBrowserSelectedAssets {type=boolean, default=true}
 CurrentAsset_Name {type=string, default=""}
 Element_LoadAsset_SetNumber_Index {type=integer}
 ElementAssetSpawned {type=Actor, default=null}
 GetVariant_PythonScript {
  type=string
  default="## Copyright Epic Games, Inc. All Rights Reserved.\r\nimport unreal\r\n\r\ndef get_asset_materials(asset_type, asset_name):\r\n    unreal.log(f\"Processing Asset: {asset_name} (Type: {asset_type})\")\r\n    ##Organizes the asset path given the asset type and the asset name\r\n    asset_path = \"&Cosmetics/\" + asset_type + \"/\" + asset_name + \".\" + asset_name\r\n    \r\n    ##Loads the asset given the path\r\n    data_asset = unreal.EditorAssetLibrary.load_asset(asset_path)\r\n\r\n    ##Gets the Material Overrides\r\n    material_overrides_length = len(data_asset.get_editor_property(\"MaterialOverrides\"))\r\n    \r\n    if True:\r\n        \r\n        try:\r\n            ##Collects the ItemVariant property from editor\r\n            properties = data_asset.get_editor_property(\"ItemVariants\")\r\n            \r\n            ##Gathers the Array of ParticleOptions\r\n            particle_options = properties[0].get_editor_property(\"ParticleOptions\")\r\n            if not particle_options:\r\n                unreal.log(f\"{asset_name} has no ParticleOptions, check for Materil options\")\r\n                particle_options = properties[0].get_editor_property(\"MaterialOptions\")\r\n            \r\n            ##Establishes the lenght so it can run through multiple variants (even if there's only one)\r\n            length = len(particle_options)\r\n\r\n            for i in range(1, length):\r\n                ##Gets the length of the VariantMaterials \r\n                particle_options_len = len(particle_options[i].get_editor_property(\"VariantMaterials\"))\r\n                if(particle_options_len == 1):\r\n                    variant_materials = particle_options[i].get_editor_property(\"VariantMaterials\")[0]\r\n                    material_override = variant_materials.get_editor_property(\"OverrideMaterial\")\r\n                    variant_materials_list.append(material_override)\r\n                    unreal.log(f\"Found material variant: {material_override.get_name()}\")\r\n                elif(particle_options_len == 2):\r\n                    variant_materials = particle_options[i].get_editor_property(\"VariantMaterials\")[0]\r\n                    variant_materials_list.append(variant_materials.get_editor_property(\"OverrideMaterial\"))\r\n                    variant_materials = particle_options[i].get_editor_property(\"VariantMaterials\")[1]\r\n                    variant_materials_list.append(variant_materials.get_editor_property(\"OverrideMaterial\"))\r\n                    ##variant_materials_list.append(override_material_list)\r\n             \r\n        except:\r\n            print(\"The Asset: \" + asset_name + \" does not have any Variant Materials or Override Materials\")\r\n\r\n    if material_overrides_length > 0:\r\n        ##Collects the property MaterialOverrides\r\n        for material_override in data_asset.get_editor_property(\"MaterialOverrides\"):\r\n            override_material = material_override.get_editor_property(\"OverrideMaterial\")\r\n            override_material_list.append(override_material)\r\n            unreal.log(f\"Found material override: {override_material.get_name()}\")\r\n\r\nvariant_materials_list = []\r\noverride_material_list = []\r\n##\"Pickaxe_PointSmoke\" \"Pickaxe_TruckScale\" \"Pickaxe_BlowWire\" \"Pickaxe_RezzPickaxe\" \"Backpack_AgileRug\" \"Glider_MedievalSheath\"\r\nasset_name = asset_name\r\nif(\"Pickaxe\" in asset_name):\r\n    get_asset_materials(\"Pickaxes\", asset_name)\r\nelif(\"Backpack\" in asset_name):\r\n    get_asset_materials(\"Backpacks\", asset_name)\r\nelif(\"Glider\" in asset_name):\r\n    get_asset_materials(\"Gliders\", asset_name)\r\nelse:\r\n    print(\"Invalid Asset Name\")"
 }
 GliderMesh_Ref {type=SkeletalMesh, default=null}
 InitialSunRotation {type=number}
 LocationActors {type="array<Actor>"}
 LocationBackup {type=Vector}
 MapAssetsStatus {default {}}
 NewLocationActors {type="array<Actor>"}
 Out Sorted Actors {type="array<Actor>"}
 override_material_list {type="array<MaterialInstanceConstant>", default []}
 PanelsArray {type="array<VerticalBox>", default []}
 PickaxeMesh_Ref {type=SkeletalMesh, default=null}
 RemoveSet {type=boolean}
 SetName {type=string, default=""}
 SetNames {type="array<string>", default []}
 SpawnBackpacks Is Checked {type=boolean, default=true}
 SpawnGliders Is Checked {type=boolean, default=true}
 spawnLocationOffSet {type=number, default=-200}
 SpawnPickaxes Is Checked {type=boolean, default=true}
 valuePrevious {type=integer, default=1}
 variant_materials_list {type="array<MaterialInstanceConstant>", default []}
 version_name {type=string, default=""}
]
widgets {
 info {rootWidgetClass=&StackBox, widgetCount=59}
 widgets [
  StackBox_0 {widgetClassPath=&StackBox, widget=StackBox_0, properties {bOverride_Cursor=true}}
  SeasonTool_VerticalBox {
   widgetClassPath=&VerticalBox
   bIsVariable=true
   parent=StackBox_0
   widget=SeasonTool_VerticalBox
   slot {class=StackBoxSlot, properties {verticalAlignment=VAlign_Top, size {sizeRule=Fill, value=1}}}
  }
  VerticalBox_0 {widgetClassPath=&VerticalBox, parent=SeasonTool_VerticalBox, widget=VerticalBox_0, slot {class=VerticalBoxSlot}}
  FreezeTime_HorizontalBox_1 {
   widgetClassPath=&HorizontalBox
   parent=VerticalBox_0
   widget=FreezeTime_HorizontalBox_1
   slot {class=VerticalBoxSlot, properties {horizontalAlignment=HAlign_Left, verticalAlignment=VAlign_Top}}
  }
  txt_SpawnSeasonAssets_4 {
   widgetClassPath=&TextBlock
   parent=FreezeTime_HorizontalBox_1
   widget=txt_SpawnSeasonAssets_4
   properties {
    text=Build Version:
    visibility=SelfHitTestInvisible
    font {
     bForceMonospaced=false
     bMaterialIsStencil=false
     fontMaterial=null
     fontObject=&Roboto
     letterSpacing=0
     monospacedWidth=1
     size=9
     skewAmount=0
     typefaceFontName=Bold
     outlineSettings {
      bApplyOutlineToDropShadows=false
      bMiteredCorners=false
      bSeparateFillAlpha=false
      outlineMaterial=null
      outlineSize=0
      outlineColor=<rgba 0 0 0 1>
     }
    }
   }
   slot {class=HorizontalBoxSlot, properties {horizontalAlignment=HAlign_Center, verticalAlignment=VAlign_Center, padding=<ltrb 4 2 4 2>}}
  }
  Overlay_0 {
   widgetClassPath=&Overlay
   parent=FreezeTime_HorizontalBox_1
   widget=Overlay_0
   slot {class=HorizontalBoxSlot, properties {verticalAlignment=VAlign_Center}}
  }
  Image_94 {
   widgetClassPath=&Image
   parent=Overlay_0
   widget=Image_94
   properties {brush {
     bIsDynamicallyLoaded=false
     drawAs=Box
     imageType=NoImage
     mirroring=NoMirror
     resourceName=None
     resourceObject=null
     tiling=NoTile
     imageSize=<32 15.7056>
     margin=<ltrb 0 0 0 0>
     outlineSettings {
      bUseBrushTransparency=false
      roundingType=HalfHeightRadius
      width=0
      color {colorUseRule=UseColor_Specified, specifiedColor=<rgba 0 0 0 0>}
      cornerRadii=<0 0 0 0>
     }
     tintColor {colorUseRule=UseColor_Specified, specifiedColor=<rgba 0.0052 0.0052 0.0052 1>}
     uVRegion {bIsValid=false, max=<0 0>, min=<0 0>}
    }}
   slot {class=OverlaySlot, properties {horizontalAlignment=HAlign_Fill, verticalAlignment=VAlign_Center}}
  }
  TimeOfDayInput {
   widgetClassPath=/Script/UMG.EditableText
   bIsVariable=true
   parent=Overlay_0
   widget=TimeOfDayInput
   properties {
    hintText=32.40
    justification=Center
    toolTipText="Set build like \"34.10\", \"33.20\", \"36.00\""
    widgetStyle {
     backgroundImageComposing {
      bIsDynamicallyLoaded=false
      drawAs=Border
      imageType=FullColor
      mirroring=NoMirror
      resourceName=../../../Engine/Content/Slate/Old/HyperlinkDotted.png
      resourceObject=null
      tiling=Both
      imageSize=<0 0>
      margin=<ltrb 0 0 0 0.1875>
      outlineSettings {
       bUseBrushTransparency=false
       roundingType=HalfHeightRadius
       width=0
       color {colorUseRule=UseColor_Specified, specifiedColor=<rgba 0 0 0 0>}
       cornerRadii=<0 0 0 0>
      }
      tintColor {colorUseRule=UseColor_Specified, specifiedColor=<rgba 1 1 1 1>}
      uVRegion {bIsValid=false, max=<0 0>, min=<0 0>}
     }
     backgroundImageSelected {
      bIsDynamicallyLoaded=false
      drawAs=Image
      imageType=NoImage
      mirroring=NoMirror
      resourceName=None
      resourceObject=null
      tiling=NoTile
      imageSize=<0 0>
      margin=<ltrb 0 0 0 0>
      outlineSettings {
       bUseBrushTransparency=false
       roundingType=HalfHeightRadius
       width=0
       color {colorUseRule=UseColor_Specified, specifiedColor=<rgba 0 0 0 0>}
       cornerRadii=<0 0 0 0>
      }
      tintColor {colorUseRule=UseColor_Specified, specifiedColor=<rgba 0 0.162 0.7454 1>}
      uVRegion {bIsValid=false, max=<0 0>, min=<0 0>}
     }
     caretImage {
      bIsDynamicallyLoaded=false
      drawAs=Image
      imageType=NoImage
      mirroring=NoMirror
      resourceName=None
      resourceObject=null
      tiling=NoTile
      imageSize=<0 0>
      margin=<ltrb 0 0 0 0>
      outlineSettings {
       bUseBrushTransparency=false
       roundingType=HalfHeightRadius
       width=0
       color {colorUseRule=UseColor_Specified, specifiedColor=<rgba 0 0 0 0>}
       cornerRadii=<0 0 0 0>
      }
      tintColor {colorUseRule=UseColor_Specified, specifiedColor=<rgba 1 1 1 1>}
      uVRegion {bIsValid=false, max=<0 0>, min=<0 0>}
     }
     colorAndOpacity {colorUseRule=UseColor_Specified, specifiedColor=<rgba 0.2051 0.5647 0.0262 1>}
     font {
      bForceMonospaced=false
      bMaterialIsStencil=false
      fontMaterial=null
      fontObject=null
      letterSpacing=0
      monospacedWidth=1
      size=10.5
      skewAmount=0
      typefaceFontName=Regular
      outlineSettings {
       bApplyOutlineToDropShadows=false
       bMiteredCorners=false
       bSeparateFillAlpha=false
       outlineMaterial=null
       outlineSize=0
       outlineColor=<rgba 0 0 0 1>
      }
     }
     hintColorAndOpacity {colorUseRule=UseColor_Foreground, specifiedColor=<rgba 1 0 1 1>}
     selectedBackgroundColor {colorUseRule=UseColor_Foreground, specifiedColor=<rgba 1 0 1 1>}
    }
   }
   slot {class=OverlaySlot, properties {verticalAlignment=VAlign_Center}}
  }
  HorizontalBox_1 {widgetClassPath=&HorizontalBox, parent=VerticalBox_0, widget=HorizontalBox_1, slot {class=VerticalBoxSlot}}
  txt_SpawnPickaxes_2 {
   widgetClassPath=&TextBlock
   parent=HorizontalBox_1
   widget=txt_SpawnPickaxes_2
   properties {
    text=Spawn Pickaxes
    visibility=SelfHitTestInvisible
    font {
     bForceMonospaced=false
     bMaterialIsStencil=false
     fontMaterial=null
     fontObject=&Roboto
     letterSpacing=0
     monospacedWidth=1
     size=9
     skewAmount=0
     typefaceFontName=Bold
     outlineSettings {
      bApplyOutlineToDropShadows=false
      bMiteredCorners=false
      bSeparateFillAlpha=false
      outlineMaterial=null
      outlineSize=0
      outlineColor=<rgba 0 0 0 1>
     }
    }
   }
   slot {class=HorizontalBoxSlot, properties {horizontalAlignment=HAlign_Center, verticalAlignment=VAlign_Center, padding=<ltrb 4 2 4 2>}}
  }
  SpawnPickaxes_CheckBox {
   widgetClassPath=&EditorUtilityCheckBox
   bIsVariable=true
   parent=HorizontalBox_1
   widget=SpawnPickaxes_CheckBox
   properties {checkedState=Checked}
   slot {class=HorizontalBoxSlot}
  }
  HorizontalBox {widgetClassPath=&HorizontalBox, parent=VerticalBox_0, widget=HorizontalBox, slot {class=VerticalBoxSlot}}
  txt_SpawnBackpacks {
   widgetClassPath=&TextBlock
   parent=HorizontalBox
   widget=txt_SpawnBackpacks
   properties {
    text=Spawn Backpacks
    visibility=SelfHitTestInvisible
    font {
     bForceMonospaced=false
     bMaterialIsStencil=false
     fontMaterial=null
     fontObject=&Roboto
     letterSpacing=0
     monospacedWidth=1
     size=9
     skewAmount=0
     typefaceFontName=Bold
     outlineSettings {
      bApplyOutlineToDropShadows=false
      bMiteredCorners=false
      bSeparateFillAlpha=false
      outlineMaterial=null
      outlineSize=0
      outlineColor=<rgba 0 0 0 1>
     }
    }
   }
   slot {class=HorizontalBoxSlot, properties {horizontalAlignment=HAlign_Center, verticalAlignment=VAlign_Center, padding=<ltrb 4 2 4 2>}}
  }
  SpawnBackpacks_CheckBox {
   widgetClassPath=&EditorUtilityCheckBox
   bIsVariable=true
   parent=HorizontalBox
   widget=SpawnBackpacks_CheckBox
   properties {checkedState=Checked}
   slot {class=HorizontalBoxSlot}
  }
  HorizontalBox_0 {widgetClassPath=&HorizontalBox, parent=VerticalBox_0, widget=HorizontalBox_0, slot {class=VerticalBoxSlot}}
  txt_SpawnGliders {
   widgetClassPath=&TextBlock
   parent=HorizontalBox_0
   widget=txt_SpawnGliders
   properties {
    text=Spawn Gliders
    visibility=SelfHitTestInvisible
    font {
     bForceMonospaced=false
     bMaterialIsStencil=false
     fontMaterial=null
     fontObject=&Roboto
     letterSpacing=0
     monospacedWidth=1
     size=9
     skewAmount=0
     typefaceFontName=Bold
     outlineSettings {
      bApplyOutlineToDropShadows=false
      bMiteredCorners=false
      bSeparateFillAlpha=false
      outlineMaterial=null
      outlineSize=0
      outlineColor=<rgba 0 0 0 1>
     }
    }
   }
   slot {class=HorizontalBoxSlot, properties {horizontalAlignment=HAlign_Center, verticalAlignment=VAlign_Center, padding=<ltrb 4 2 4 2>}}
  }
  SpawnGliders_CheckBox {
   widgetClassPath=&EditorUtilityCheckBox
   bIsVariable=true
   parent=HorizontalBox_0
   widget=SpawnGliders_CheckBox
   properties {checkedState=Checked}
   slot {class=HorizontalBoxSlot}
  }
  Spacer {
   widgetClassPath=&Spacer
   parent=SeasonTool_VerticalBox
   widget=Spacer
   properties {size=<1 5>}
   slot {class=VerticalBoxSlot, properties {verticalAlignment=VAlign_Center}}
  }
  HorizontalBox_2 {
   widgetClassPath=&HorizontalBox
   parent=SeasonTool_VerticalBox
   widget=HorizontalBox_2
   slot {class=VerticalBoxSlot, properties {verticalAlignment=VAlign_Center, padding=<ltrb 4 0 4 0>}}
  }
  btn_SpawnSeasonAssets {
   widgetClassPath=&EditorUtilityButton
   parent=HorizontalBox_2
   widget=btn_SpawnSeasonAssets
   properties {
    toolTipText=This button will spawn every ID for a specific Build Version
    widgetStyle {
     clickedSlateSound {resourceObject=null}
     disabled {
      bIsDynamicallyLoaded=false
      drawAs=RoundedBox
      imageType=NoImage
      mirroring=NoMirror
      resourceName=None
      resourceObject=null
      tiling=NoTile
      imageSize=<0 0>
      margin=<ltrb 0 0 0 0>
      outlineSettings {
       bUseBrushTransparency=false
       roundingType=FixedRadius
       width=1
       color {colorUseRule=UseColor_ColorTable, specifiedColor=<rgba 0 0 0 0>}
       cornerRadii=<4 4 4 4>
      }
      tintColor {colorUseRule=UseColor_Specified, specifiedColor=<rgba 0.0395 0.0395 0.0395 1>}
      uVRegion {bIsValid=false, max=<0 0>, min=<0 0>}
     }
     disabledForeground {colorUseRule=UseColor_ColorTable, specifiedColor=<rgba 0 0 0 0>}
     hovered {
      bIsDynamicallyLoaded=false
      drawAs=RoundedBox
      imageType=NoImage
      mirroring=NoMirror
      resourceName=None
      resourceObject=null
      tiling=NoTile
      imageSize=<32 32>
      margin=<ltrb 0 0 0 0>
      outlineSettings {
       bUseBrushTransparency=false
       roundingType=FixedRadius
       width=1
       color {colorUseRule=UseColor_ColorTable, specifiedColor=<rgba 0 0 0 0>}
       cornerRadii=<4 4 4 4>
      }
      tintColor {colorUseRule=UseColor_Specified, specifiedColor=<rgba 0.0953 0.0953 0.0953 1>}
      uVRegion {bIsValid=false, max=<0 0>, min=<0 0>}
     }
     hoveredForeground {colorUseRule=UseColor_ColorTable, specifiedColor=<rgba 0 0 0 0>}
     hoveredSlateSound {resourceObject=null}
     normal {
      bIsDynamicallyLoaded=false
      drawAs=RoundedBox
      imageType=NoImage
      mirroring=NoMirror
      resourceName=None
      resourceObject=null
      tiling=NoTile
      imageSize=<32 32>
      margin=<ltrb 0 0 0 0>
      outlineSettings {
       bUseBrushTransparency=false
       roundingType=FixedRadius
       width=1
       color {colorUseRule=UseColor_ColorTable, specifiedColor=<rgba 0 0 0 0>}
       cornerRadii=<4 4 4 4>
      }
      tintColor {colorUseRule=UseColor_Specified, specifiedColor=<rgba 0.2057 0.5656 0.0265 1>}
      uVRegion {bIsValid=false, max=<0 0>, min=<0 0>}
     }
     normalForeground {colorUseRule=UseColor_ColorTable, specifiedColor=<rgba 0 0 0 0>}
     normalPadding=<ltrb 12 1.5 12 1.5>
     pressed {
      bIsDynamicallyLoaded=false
      drawAs=RoundedBox
      imageType=NoImage
      mirroring=NoMirror
      resourceName=None
      resourceObject=null
      tiling=NoTile
      imageSize=<32 32>
      margin=<ltrb 0 0 0 0>
      outlineSettings {
       bUseBrushTransparency=false
       roundingType=FixedRadius
       width=1
       color {colorUseRule=UseColor_ColorTable, specifiedColor=<rgba 0 0 0 0>}
       cornerRadii=<4 4 4 4>
      }
      tintColor {colorUseRule=UseColor_Specified, specifiedColor=<rgba 0.0284 0.0284 0.0284 1>}
      uVRegion {bIsValid=false, max=<0 0>, min=<0 0>}
     }
     pressedForeground {colorUseRule=UseColor_ColorTable, specifiedColor=<rgba 0 0 0 0>}
     pressedPadding=<ltrb 12 2.5 12 0.5>
     pressedSlateSound {resourceObject=null}
    }
   }
   slot {class=HorizontalBoxSlot, properties {verticalAlignment=VAlign_Center, size {sizeRule=Fill, value=1}}}
  }
  ScaleBox_2 {widgetClassPath=&ScaleBox, parent=btn_SpawnSeasonAssets, widget=ScaleBox_2, slot {class=ButtonSlot}}
  txt_SpawnSeasonAssets {
   widgetClassPath=&TextBlock
   parent=ScaleBox_2
   widget=txt_SpawnSeasonAssets
   properties {
    text=Spawn Season Assets
    visibility=SelfHitTestInvisible
    colorAndOpacity {colorUseRule=UseColor_Specified, specifiedColor=<rgba 0 0 0 1>}
    font {
     bForceMonospaced=false
     bMaterialIsStencil=false
     fontMaterial=null
     fontObject=&Roboto
     letterSpacing=0
     monospacedWidth=1
     size=11.25
     skewAmount=0
     typefaceFontName=Bold
     outlineSettings {
      bApplyOutlineToDropShadows=false
      bMiteredCorners=false
      bSeparateFillAlpha=false
      outlineMaterial=null
      outlineSize=0
      outlineColor=<rgba 0 0 0 1>
     }
    }
   }
   slot {class=ScaleBoxSlot}
  }
  Spacer_4 {
   widgetClassPath=&Spacer
   parent=HorizontalBox_2
   widget=Spacer_4
   properties {size=<10 0>}
   slot {class=HorizontalBoxSlot, properties {verticalAlignment=VAlign_Center}}
  }
  btn_DeleteSeasonAssets {
   widgetClassPath=&EditorUtilityButton
   parent=HorizontalBox_2
   widget=btn_DeleteSeasonAssets
   properties {widgetStyle {
     clickedSlateSound {resourceObject=null}
     disabled {
      bIsDynamicallyLoaded=false
      drawAs=RoundedBox
      imageType=NoImage
      mirroring=NoMirror
      resourceName=None
      resourceObject=null
      tiling=NoTile
      imageSize=<0 0>
      margin=<ltrb 0 0 0 0>
      outlineSettings {
       bUseBrushTransparency=false
       roundingType=FixedRadius
       width=1
       color {colorUseRule=UseColor_ColorTable, specifiedColor=<rgba 0 0 0 0>}
       cornerRadii=<4 4 4 4>
      }
      tintColor {colorUseRule=UseColor_Specified, specifiedColor=<rgba 0.0395 0.0395 0.0395 1>}
      uVRegion {bIsValid=false, max=<0 0>, min=<0 0>}
     }
     disabledForeground {colorUseRule=UseColor_ColorTable, specifiedColor=<rgba 0 0 0 0>}
     hovered {
      bIsDynamicallyLoaded=false
      drawAs=RoundedBox
      imageType=NoImage
      mirroring=NoMirror
      resourceName=None
      resourceObject=null
      tiling=NoTile
      imageSize=<32 32>
      margin=<ltrb 0 0 0 0>
      outlineSettings {
       bUseBrushTransparency=false
       roundingType=FixedRadius
       width=1
       color {colorUseRule=UseColor_ColorTable, specifiedColor=<rgba 0 0 0 0>}
       cornerRadii=<4 4 4 4>
      }
      tintColor {colorUseRule=UseColor_Specified, specifiedColor=<rgba 0.0953 0.0953 0.0953 1>}
      uVRegion {bIsValid=false, max=<0 0>, min=<0 0>}
     }
     hoveredForeground {colorUseRule=UseColor_ColorTable, specifiedColor=<rgba 0 0 0 0>}
     hoveredSlateSound {resourceObject=null}
     normal {
      bIsDynamicallyLoaded=false
      drawAs=RoundedBox
      imageType=NoImage
      mirroring=NoMirror
      resourceName=None
      resourceObject=null
      tiling=NoTile
      imageSize=<32 32>
      margin=<ltrb 0 0 0 0>
      outlineSettings {
       bUseBrushTransparency=false
       roundingType=FixedRadius
       width=1
       color {colorUseRule=UseColor_ColorTable, specifiedColor=<rgba 0 0 0 0>}
       cornerRadii=<4 4 4 4>
      }
      tintColor {colorUseRule=UseColor_Specified, specifiedColor=<rgba 0.4365 0.0332 0.0332 1>}
      uVRegion {bIsValid=false, max=<0 0>, min=<0 0>}
     }
     normalForeground {colorUseRule=UseColor_ColorTable, specifiedColor=<rgba 0 0 0 0>}
     normalPadding=<ltrb 12 1.5 12 1.5>
     pressed {
      bIsDynamicallyLoaded=false
      drawAs=RoundedBox
      imageType=NoImage
      mirroring=NoMirror
      resourceName=None
      resourceObject=null
      tiling=NoTile
      imageSize=<32 32>
      margin=<ltrb 0 0 0 0>
      outlineSettings {
       bUseBrushTransparency=false
       roundingType=FixedRadius
       width=1
       color {colorUseRule=UseColor_ColorTable, specifiedColor=<rgba 0 0 0 0>}
       cornerRadii=<4 4 4 4>
      }
      tintColor {colorUseRule=UseColor_Specified, specifiedColor=<rgba 0.0284 0.0284 0.0284 1>}
      uVRegion {bIsValid=false, max=<0 0>, min=<0 0>}
     }
     pressedForeground {colorUseRule=UseColor_ColorTable, specifiedColor=<rgba 0 0 0 0>}
     pressedPadding=<ltrb 12 2.5 12 0.5>
     pressedSlateSound {resourceObject=null}
    }}
   slot {class=HorizontalBoxSlot, properties {verticalAlignment=VAlign_Center, size {sizeRule=Fill, value=1}}}
  }
  ScaleBox_3 {widgetClassPath=&ScaleBox, parent=btn_DeleteSeasonAssets, widget=ScaleBox_3, slot {class=ButtonSlot}}
  txt_DeleteSeasonAssets {
   widgetClassPath=&TextBlock
   parent=ScaleBox_3
   widget=txt_DeleteSeasonAssets
   properties {
    text=Delete Season Assets
    visibility=SelfHitTestInvisible
    colorAndOpacity {colorUseRule=UseColor_Specified, specifiedColor=<rgba 0 0 0 1>}
    font {
     bForceMonospaced=false
     bMaterialIsStencil=false
     fontMaterial=null
     fontObject=&Roboto
     letterSpacing=0
     monospacedWidth=1
     size=11.25
     skewAmount=0
     typefaceFontName=Bold
     outlineSettings {
      bApplyOutlineToDropShadows=false
      bMiteredCorners=false
      bSeparateFillAlpha=false
      outlineMaterial=null
      outlineSize=0
      outlineColor=<rgba 0 0 0 1>
     }
    }
   }
   slot {class=ScaleBoxSlot}
  }
  Spacer_3 {
   widgetClassPath=&Spacer
   parent=SeasonTool_VerticalBox
   widget=Spacer_3
   properties {size=<1 12>}
   slot {class=VerticalBoxSlot, properties {verticalAlignment=VAlign_Center}}
  }
  HorizontalBox_4 {
   widgetClassPath=&HorizontalBox
   parent=SeasonTool_VerticalBox
   widget=HorizontalBox_4
   slot {class=VerticalBoxSlot, properties {verticalAlignment=VAlign_Center, padding=<ltrb 4 0 4 0>}}
  }
  btn_SpawnSelectedAssets {
   widgetClassPath=&EditorUtilityButton
   parent=HorizontalBox_4
   widget=btn_SpawnSelectedAssets
   properties {
    toolTipText=This button will spawn the selected IDs in the current content browser window.
    widgetStyle {
     clickedSlateSound {resourceObject=null}
     disabled {
      bIsDynamicallyLoaded=false
      drawAs=RoundedBox
      imageType=NoImage
      mirroring=NoMirror
      resourceName=None
      resourceObject=null
      tiling=NoTile
      imageSize=<0 0>
      margin=<ltrb 0 0 0 0>
      outlineSettings {
       bUseBrushTransparency=false
       roundingType=FixedRadius
       width=1
       color {colorUseRule=UseColor_ColorTable, specifiedColor=<rgba 0 0 0 0>}
       cornerRadii=<4 4 4 4>
      }
      tintColor {colorUseRule=UseColor_Specified, specifiedColor=<rgba 0.0395 0.0395 0.0395 1>}
      uVRegion {bIsValid=false, max=<0 0>, min=<0 0>}
     }
     disabledForeground {colorUseRule=UseColor_ColorTable, specifiedColor=<rgba 0 0 0 0>}
     hovered {
      bIsDynamicallyLoaded=false
      drawAs=RoundedBox
      imageType=NoImage
      mirroring=NoMirror
      resourceName=None
      resourceObject=null
      tiling=NoTile
      imageSize=<32 32>
      margin=<ltrb 0 0 0 0>
      outlineSettings {
       bUseBrushTransparency=false
       roundingType=FixedRadius
       width=1
       color {colorUseRule=UseColor_ColorTable, specifiedColor=<rgba 0 0 0 0>}
       cornerRadii=<4 4 4 4>
      }
      tintColor {colorUseRule=UseColor_Specified, specifiedColor=<rgba 0.0953 0.0953 0.0953 1>}
      uVRegion {bIsValid=false, max=<0 0>, min=<0 0>}
     }
     hoveredForeground {colorUseRule=UseColor_ColorTable, specifiedColor=<rgba 0 0 0 0>}
     hoveredSlateSound {resourceObject=null}
     normal {
      bIsDynamicallyLoaded=false
      drawAs=RoundedBox
      imageType=NoImage
      mirroring=NoMirror
      resourceName=None
      resourceObject=null
      tiling=NoTile
      imageSize=<32 32>
      margin=<ltrb 0 0 0 0>
      outlineSettings {
       bUseBrushTransparency=false
       roundingType=FixedRadius
       width=1
       color {colorUseRule=UseColor_ColorTable, specifiedColor=<rgba 0 0 0 0>}
       cornerRadii=<4 4 4 4>
      }
      tintColor {colorUseRule=UseColor_Specified, specifiedColor=<rgba 0.5656 0.5656 0.5656 1>}
      uVRegion {bIsValid=false, max=<0 0>, min=<0 0>}
     }
     normalForeground {colorUseRule=UseColor_ColorTable, specifiedColor=<rgba 0 0 0 0>}
     normalPadding=<ltrb 12 1.5 12 1.5>
     pressed {
      bIsDynamicallyLoaded=false
      drawAs=RoundedBox
      imageType=NoImage
      mirroring=NoMirror
      resourceName=None
      resourceObject=null
      tiling=NoTile
      imageSize=<32 32>
      margin=<ltrb 0 0 0 0>
      outlineSettings {
       bUseBrushTransparency=false
       roundingType=FixedRadius
       width=1
       color {colorUseRule=UseColor_ColorTable, specifiedColor=<rgba 0 0 0 0>}
       cornerRadii=<4 4 4 4>
      }
      tintColor {colorUseRule=UseColor_Specified, specifiedColor=<rgba 0.0284 0.0284 0.0284 1>}
      uVRegion {bIsValid=false, max=<0 0>, min=<0 0>}
     }
     pressedForeground {colorUseRule=UseColor_ColorTable, specifiedColor=<rgba 0 0 0 0>}
     pressedPadding=<ltrb 12 2.5 12 0.5>
     pressedSlateSound {resourceObject=null}
    }
   }
   slot {class=HorizontalBoxSlot, properties {verticalAlignment=VAlign_Center, size {sizeRule=Fill, value=1}}}
  }
  ScaleBox {widgetClassPath=&ScaleBox, parent=btn_SpawnSelectedAssets, widget=ScaleBox, slot {class=ButtonSlot}}
  txt_SpawnSelected {
   widgetClassPath=&TextBlock
   parent=ScaleBox
   widget=txt_SpawnSelected
   properties {
    text=Spawn Selected Assets
    visibility=SelfHitTestInvisible
    colorAndOpacity {colorUseRule=UseColor_Specified, specifiedColor=<rgba 0 0 0 1>}
    font {
     bForceMonospaced=false
     bMaterialIsStencil=false
     fontMaterial=null
     fontObject=&Roboto
     letterSpacing=0
     monospacedWidth=1
     size=10.5
     skewAmount=0
     typefaceFontName=Bold
     outlineSettings {
      bApplyOutlineToDropShadows=false
      bMiteredCorners=false
      bSeparateFillAlpha=false
      outlineMaterial=null
      outlineSize=0
      outlineColor=<rgba 0 0 0 1>
     }
    }
   }
   slot {class=ScaleBoxSlot}
  }
  Spacer_6 {
   widgetClassPath=&Spacer
   parent=HorizontalBox_4
   widget=Spacer_6
   properties {size=<10 0>}
   slot {class=HorizontalBoxSlot, properties {verticalAlignment=VAlign_Center}}
  }
  btn_DisplaySelectedAssets {
   widgetClassPath=&EditorUtilityButton
   parent=HorizontalBox_4
   widget=btn_DisplaySelectedAssets
   properties {
    toolTipText=This button will spawn the selected IDs in the current content browser window.
    widgetStyle {
     clickedSlateSound {resourceObject=null}
     disabled {
      bIsDynamicallyLoaded=false
      drawAs=RoundedBox
      imageType=NoImage
      mirroring=NoMirror
      resourceName=None
      resourceObject=null
      tiling=NoTile
      imageSize=<0 0>
      margin=<ltrb 0 0 0 0>
      outlineSettings {
       bUseBrushTransparency=false
       roundingType=FixedRadius
       width=1
       color {colorUseRule=UseColor_ColorTable, specifiedColor=<rgba 0 0 0 0>}
       cornerRadii=<4 4 4 4>
      }
      tintColor {colorUseRule=UseColor_Specified, specifiedColor=<rgba 0.0395 0.0395 0.0395 1>}
      uVRegion {bIsValid=false, max=<0 0>, min=<0 0>}
     }
     disabledForeground {colorUseRule=UseColor_ColorTable, specifiedColor=<rgba 0 0 0 0>}
     hovered {
      bIsDynamicallyLoaded=false
      drawAs=RoundedBox
      imageType=NoImage
      mirroring=NoMirror
      resourceName=None
      resourceObject=null
      tiling=NoTile
      imageSize=<32 32>
      margin=<ltrb 0 0 0 0>
      outlineSettings {
       bUseBrushTransparency=false
       roundingType=FixedRadius
       width=1
       color {colorUseRule=UseColor_ColorTable, specifiedColor=<rgba 0 0 0 0>}
       cornerRadii=<4 4 4 4>
      }
      tintColor {colorUseRule=UseColor_Specified, specifiedColor=<rgba 0.0953 0.0953 0.0953 1>}
      uVRegion {bIsValid=false, max=<0 0>, min=<0 0>}
     }
     hoveredForeground {colorUseRule=UseColor_ColorTable, specifiedColor=<rgba 0 0 0 0>}
     hoveredSlateSound {resourceObject=null}
     normal {
      bIsDynamicallyLoaded=false
      drawAs=RoundedBox
      imageType=NoImage
      mirroring=NoMirror
      resourceName=None
      resourceObject=null
      tiling=NoTile
      imageSize=<32 32>
      margin=<ltrb 0 0 0 0>
      outlineSettings {
       bUseBrushTransparency=false
       roundingType=FixedRadius
       width=1
       color {colorUseRule=UseColor_ColorTable, specifiedColor=<rgba 0 0 0 0>}
       cornerRadii=<4 4 4 4>
      }
      tintColor {colorUseRule=UseColor_Specified, specifiedColor=<rgba 0.5656 0.5656 0.5656 1>}
      uVRegion {bIsValid=false, max=<0 0>, min=<0 0>}
     }
     normalForeground {colorUseRule=UseColor_ColorTable, specifiedColor=<rgba 0 0 0 0>}
     normalPadding=<ltrb 12 1.5 12 1.5>
     pressed {
      bIsDynamicallyLoaded=false
      drawAs=RoundedBox
      imageType=NoImage
      mirroring=NoMirror
      resourceName=None
      resourceObject=null
      tiling=NoTile
      imageSize=<32 32>
      margin=<ltrb 0 0 0 0>
      outlineSettings {
       bUseBrushTransparency=false
       roundingType=FixedRadius
       width=1
       color {colorUseRule=UseColor_ColorTable, specifiedColor=<rgba 0 0 0 0>}
       cornerRadii=<4 4 4 4>
      }
      tintColor {colorUseRule=UseColor_Specified, specifiedColor=<rgba 0.0284 0.0284 0.0284 1>}
      uVRegion {bIsValid=false, max=<0 0>, min=<0 0>}
     }
     pressedForeground {colorUseRule=UseColor_ColorTable, specifiedColor=<rgba 0 0 0 0>}
     pressedPadding=<ltrb 12 2.5 12 0.5>
     pressedSlateSound {resourceObject=null}
    }
   }
   slot {class=HorizontalBoxSlot, properties {verticalAlignment=VAlign_Center, size {sizeRule=Fill, value=1}}}
  }
  ScaleBox_8 {widgetClassPath=&ScaleBox, parent=btn_DisplaySelectedAssets, widget=ScaleBox_8, slot {class=ButtonSlot}}
  txt_DisplaySelectedItems {
   widgetClassPath=&TextBlock
   parent=ScaleBox_8
   widget=txt_DisplaySelectedItems
   properties {
    text=Display Selected Assets
    visibility=SelfHitTestInvisible
    colorAndOpacity {colorUseRule=UseColor_Specified, specifiedColor=<rgba 0 0 0 1>}
    font {
     bForceMonospaced=false
     bMaterialIsStencil=false
     fontMaterial=null
     fontObject=&Roboto
     letterSpacing=0
     monospacedWidth=1
     size=10.5
     skewAmount=0
     typefaceFontName=Bold
     outlineSettings {
      bApplyOutlineToDropShadows=false
      bMiteredCorners=false
      bSeparateFillAlpha=false
      outlineMaterial=null
      outlineSize=0
      outlineColor=<rgba 0 0 0 1>
     }
    }
   }
   slot {class=ScaleBoxSlot}
  }
  Spacer_2 {
   widgetClassPath=&Spacer
   parent=SeasonTool_VerticalBox
   widget=Spacer_2
   properties {size=<1 12>}
   slot {class=VerticalBoxSlot, properties {verticalAlignment=VAlign_Center}}
  }
  HorizontalBox_3 {
   widgetClassPath=&HorizontalBox
   parent=SeasonTool_VerticalBox
   widget=HorizontalBox_3
   slot {class=VerticalBoxSlot, properties {verticalAlignment=VAlign_Center, padding=<ltrb 4 0 4 0>}}
  }
  btn_KeepSet {
   widgetClassPath=&EditorUtilityButton
   parent=HorizontalBox_3
   widget=btn_KeepSet
   properties {widgetStyle {
     clickedSlateSound {resourceObject=null}
     disabled {
      bIsDynamicallyLoaded=false
      drawAs=RoundedBox
      imageType=NoImage
      mirroring=NoMirror
      resourceName=None
      resourceObject=null
      tiling=NoTile
      imageSize=<0 0>
      margin=<ltrb 0 0 0 0>
      outlineSettings {
       bUseBrushTransparency=false
       roundingType=FixedRadius
       width=1
       color {colorUseRule=UseColor_ColorTable, specifiedColor=<rgba 0 0 0 0>}
       cornerRadii=<4 4 4 4>
      }
      tintColor {colorUseRule=UseColor_Specified, specifiedColor=<rgba 0.0395 0.0395 0.0395 1>}
      uVRegion {bIsValid=false, max=<0 0>, min=<0 0>}
     }
     disabledForeground {colorUseRule=UseColor_ColorTable, specifiedColor=<rgba 0 0 0 0>}
     hovered {
      bIsDynamicallyLoaded=false
      drawAs=RoundedBox
      imageType=NoImage
      mirroring=NoMirror
      resourceName=None
      resourceObject=null
      tiling=NoTile
      imageSize=<32 32>
      margin=<ltrb 0 0 0 0>
      outlineSettings {
       bUseBrushTransparency=false
       roundingType=FixedRadius
       width=1
       color {colorUseRule=UseColor_ColorTable, specifiedColor=<rgba 0 0 0 0>}
       cornerRadii=<4 4 4 4>
      }
      tintColor {colorUseRule=UseColor_Specified, specifiedColor=<rgba 0.0953 0.0953 0.0953 1>}
      uVRegion {bIsValid=false, max=<0 0>, min=<0 0>}
     }
     hoveredForeground {colorUseRule=UseColor_ColorTable, specifiedColor=<rgba 0 0 0 0>}
     hoveredSlateSound {resourceObject=null}
     normal {
      bIsDynamicallyLoaded=false
      drawAs=RoundedBox
      imageType=NoImage
      mirroring=NoMirror
      resourceName=None
      resourceObject=null
      tiling=NoTile
      imageSize=<32 32>
      margin=<ltrb 0 0 0 0>
      outlineSettings {
       bUseBrushTransparency=false
       roundingType=FixedRadius
       width=1
       color {colorUseRule=UseColor_ColorTable, specifiedColor=<rgba 0 0 0 0>}
       cornerRadii=<4 4 4 4>
      }
      tintColor {colorUseRule=UseColor_Specified, specifiedColor=<rgba 0.5647 0.5647 0.5647 1>}
      uVRegion {bIsValid=false, max=<0 0>, min=<0 0>}
     }
     normalForeground {colorUseRule=UseColor_ColorTable, specifiedColor=<rgba 0 0 0 0>}
     normalPadding=<ltrb 12 1.5 12 1.5>
     pressed {
      bIsDynamicallyLoaded=false
      drawAs=RoundedBox
      imageType=NoImage
      mirroring=NoMirror
      resourceName=None
      resourceObject=null
      tiling=NoTile
      imageSize=<32 32>
      margin=<ltrb 0 0 0 0>
      outlineSettings {
       bUseBrushTransparency=false
       roundingType=FixedRadius
       width=1
       color {colorUseRule=UseColor_ColorTable, specifiedColor=<rgba 0 0 0 0>}
       cornerRadii=<4 4 4 4>
      }
      tintColor {colorUseRule=UseColor_Specified, specifiedColor=<rgba 0.0284 0.0284 0.0284 1>}
      uVRegion {bIsValid=false, max=<0 0>, min=<0 0>}
     }
     pressedForeground {colorUseRule=UseColor_ColorTable, specifiedColor=<rgba 0 0 0 0>}
     pressedPadding=<ltrb 12 2.5 12 0.5>
     pressedSlateSound {resourceObject=null}
    }}
   slot {class=HorizontalBoxSlot, properties {verticalAlignment=VAlign_Center, size {sizeRule=Fill, value=1}}}
  }
  ScaleBox_4 {widgetClassPath=&ScaleBox, parent=btn_KeepSet, widget=ScaleBox_4, slot {class=ButtonSlot}}
  txt_KeepSet {
   widgetClassPath=&TextBlock
   parent=ScaleBox_4
   widget=txt_KeepSet
   properties {
    text=Keep Set
    visibility=SelfHitTestInvisible
    colorAndOpacity {colorUseRule=UseColor_Specified, specifiedColor=<rgba 0 0 0 1>}
    font {
     bForceMonospaced=false
     bMaterialIsStencil=false
     fontMaterial=null
     fontObject=&Roboto
     letterSpacing=0
     monospacedWidth=1
     size=10.5
     skewAmount=0
     typefaceFontName=Bold
     outlineSettings {
      bApplyOutlineToDropShadows=false
      bMiteredCorners=false
      bSeparateFillAlpha=false
      outlineMaterial=null
      outlineSize=0
      outlineColor=<rgba 0 0 0 1>
     }
    }
   }
   slot {class=ScaleBoxSlot}
  }
  Spacer_5 {
   widgetClassPath=&Spacer
   parent=HorizontalBox_3
   widget=Spacer_5
   properties {size=<10 0>}
   slot {class=HorizontalBoxSlot, properties {verticalAlignment=VAlign_Center}}
  }
  btn_HideSet {
   widgetClassPath=&EditorUtilityButton
   parent=HorizontalBox_3
   widget=btn_HideSet
   properties {widgetStyle {
     clickedSlateSound {resourceObject=null}
     disabled {
      bIsDynamicallyLoaded=false
      drawAs=RoundedBox
      imageType=NoImage
      mirroring=NoMirror
      resourceName=None
      resourceObject=null
      tiling=NoTile
      imageSize=<0 0>
      margin=<ltrb 0 0 0 0>
      outlineSettings {
       bUseBrushTransparency=false
       roundingType=FixedRadius
       width=1
       color {colorUseRule=UseColor_ColorTable, specifiedColor=<rgba 0 0 0 0>}
       cornerRadii=<4 4 4 4>
      }
      tintColor {colorUseRule=UseColor_Specified, specifiedColor=<rgba 0.0395 0.0395 0.0395 1>}
      uVRegion {bIsValid=false, max=<0 0>, min=<0 0>}
     }
     disabledForeground {colorUseRule=UseColor_ColorTable, specifiedColor=<rgba 0 0 0 0>}
     hovered {
      bIsDynamicallyLoaded=false
      drawAs=RoundedBox
      imageType=NoImage
      mirroring=NoMirror
      resourceName=None
      resourceObject=null
      tiling=NoTile
      imageSize=<32 32>
      margin=<ltrb 0 0 0 0>
      outlineSettings {
       bUseBrushTransparency=false
       roundingType=FixedRadius
       width=1
       color {colorUseRule=UseColor_ColorTable, specifiedColor=<rgba 0 0 0 0>}
       cornerRadii=<4 4 4 4>
      }
      tintColor {colorUseRule=UseColor_Specified, specifiedColor=<rgba 0.0953 0.0953 0.0953 1>}
      uVRegion {bIsValid=false, max=<0 0>, min=<0 0>}
     }
     hoveredForeground {colorUseRule=UseColor_ColorTable, specifiedColor=<rgba 0 0 0 0>}
     hoveredSlateSound {resourceObject=null}
     normal {
      bIsDynamicallyLoaded=false
      drawAs=RoundedBox
      imageType=NoImage
      mirroring=NoMirror
      resourceName=None
      resourceObject=null
      tiling=NoTile
      imageSize=<32 32>
      margin=<ltrb 0 0 0 0>
      outlineSettings {
       bUseBrushTransparency=false
       roundingType=FixedRadius
       width=1
       color {colorUseRule=UseColor_ColorTable, specifiedColor=<rgba 0 0 0 0>}
       cornerRadii=<4 4 4 4>
      }
      tintColor {colorUseRule=UseColor_Specified, specifiedColor=<rgba 0.5647 0.5647 0.5647 1>}
      uVRegion {bIsValid=false, max=<0 0>, min=<0 0>}
     }
     normalForeground {colorUseRule=UseColor_ColorTable, specifiedColor=<rgba 0 0 0 0>}
     normalPadding=<ltrb 12 1.5 12 1.5>
     pressed {
      bIsDynamicallyLoaded=false
      drawAs=RoundedBox
      imageType=NoImage
      mirroring=NoMirror
      resourceName=None
      resourceObject=null
      tiling=NoTile
      imageSize=<32 32>
      margin=<ltrb 0 0 0 0>
      outlineSettings {
       bUseBrushTransparency=false
       roundingType=FixedRadius
       width=1
       color {colorUseRule=UseColor_ColorTable, specifiedColor=<rgba 0 0 0 0>}
       cornerRadii=<4 4 4 4>
      }
      tintColor {colorUseRule=UseColor_Specified, specifiedColor=<rgba 0.0284 0.0284 0.0284 1>}
      uVRegion {bIsValid=false, max=<0 0>, min=<0 0>}
     }
     pressedForeground {colorUseRule=UseColor_ColorTable, specifiedColor=<rgba 0 0 0 0>}
     pressedPadding=<ltrb 12 2.5 12 0.5>
     pressedSlateSound {resourceObject=null}
    }}
   slot {class=HorizontalBoxSlot, properties {verticalAlignment=VAlign_Center, size {sizeRule=Fill, value=1}}}
  }
  ScaleBox_5 {widgetClassPath=&ScaleBox, parent=btn_HideSet, widget=ScaleBox_5, slot {class=ButtonSlot}}
  txt_HideSet {
   widgetClassPath=&TextBlock
   parent=ScaleBox_5
   widget=txt_HideSet
   properties {
    text=Hide Set
    visibility=SelfHitTestInvisible
    colorAndOpacity {colorUseRule=UseColor_Specified, specifiedColor=<rgba 0 0 0 1>}
    font {
     bForceMonospaced=false
     bMaterialIsStencil=false
     fontMaterial=null
     fontObject=&Roboto
     letterSpacing=0
     monospacedWidth=1
     size=10.5
     skewAmount=0
     typefaceFontName=Bold
     outlineSettings {
      bApplyOutlineToDropShadows=false
      bMiteredCorners=false
      bSeparateFillAlpha=false
      outlineMaterial=null
      outlineSize=0
      outlineColor=<rgba 0 0 0 1>
     }
    }
   }
   slot {class=ScaleBoxSlot}
  }
  Spacer_1 {
   widgetClassPath=&Spacer
   parent=HorizontalBox_3
   widget=Spacer_1
   properties {size=<10 0>}
   slot {class=HorizontalBoxSlot, properties {verticalAlignment=VAlign_Center}}
  }
  btn_ShowAll {
   widgetClassPath=&EditorUtilityButton
   parent=HorizontalBox_3
   widget=btn_ShowAll
   properties {widgetStyle {
     clickedSlateSound {resourceObject=null}
     disabled {
      bIsDynamicallyLoaded=false
      drawAs=RoundedBox
      imageType=NoImage
      mirroring=NoMirror
      resourceName=None
      resourceObject=null
      tiling=NoTile
      imageSize=<0 0>
      margin=<ltrb 0 0 0 0>
      outlineSettings {
       bUseBrushTransparency=false
       roundingType=FixedRadius
       width=1
       color {colorUseRule=UseColor_ColorTable, specifiedColor=<rgba 0 0 0 0>}
       cornerRadii=<4 4 4 4>
      }
      tintColor {colorUseRule=UseColor_Specified, specifiedColor=<rgba 0.0395 0.0395 0.0395 1>}
      uVRegion {bIsValid=false, max=<0 0>, min=<0 0>}
     }
     disabledForeground {colorUseRule=UseColor_ColorTable, specifiedColor=<rgba 0 0 0 0>}
     hovered {
      bIsDynamicallyLoaded=false
      drawAs=RoundedBox
      imageType=NoImage
      mirroring=NoMirror
      resourceName=None
      resourceObject=null
      tiling=NoTile
      imageSize=<32 32>
      margin=<ltrb 0 0 0 0>
      outlineSettings {
       bUseBrushTransparency=false
       roundingType=FixedRadius
       width=1
       color {colorUseRule=UseColor_ColorTable, specifiedColor=<rgba 0 0 0 0>}
       cornerRadii=<4 4 4 4>
      }
      tintColor {colorUseRule=UseColor_Specified, specifiedColor=<rgba 0.0953 0.0953 0.0953 1>}
      uVRegion {bIsValid=false, max=<0 0>, min=<0 0>}
     }
     hoveredForeground {colorUseRule=UseColor_ColorTable, specifiedColor=<rgba 0 0 0 0>}
     hoveredSlateSound {resourceObject=null}
     normal {
      bIsDynamicallyLoaded=false
      drawAs=RoundedBox
      imageType=NoImage
      mirroring=NoMirror
      resourceName=None
      resourceObject=null
      tiling=NoTile
      imageSize=<32 32>
      margin=<ltrb 0 0 0 0>
      outlineSettings {
       bUseBrushTransparency=false
       roundingType=FixedRadius
       width=1
       color {colorUseRule=UseColor_ColorTable, specifiedColor=<rgba 0 0 0 0>}
       cornerRadii=<4 4 4 4>
      }
      tintColor {colorUseRule=UseColor_Specified, specifiedColor=<rgba 0.5647 0.5647 0.5647 1>}
      uVRegion {bIsValid=false, max=<0 0>, min=<0 0>}
     }
     normalForeground {colorUseRule=UseColor_ColorTable, specifiedColor=<rgba 0 0 0 0>}
     normalPadding=<ltrb 12 1.5 12 1.5>
     pressed {
      bIsDynamicallyLoaded=false
      drawAs=RoundedBox
      imageType=NoImage
      mirroring=NoMirror
      resourceName=None
      resourceObject=null
      tiling=NoTile
      imageSize=<32 32>
      margin=<ltrb 0 0 0 0>
      outlineSettings {
       bUseBrushTransparency=false
       roundingType=FixedRadius
       width=1
       color {colorUseRule=UseColor_ColorTable, specifiedColor=<rgba 0 0 0 0>}
       cornerRadii=<4 4 4 4>
      }
      tintColor {colorUseRule=UseColor_Specified, specifiedColor=<rgba 0.0284 0.0284 0.0284 1>}
      uVRegion {bIsValid=false, max=<0 0>, min=<0 0>}
     }
     pressedForeground {colorUseRule=UseColor_ColorTable, specifiedColor=<rgba 0 0 0 0>}
     pressedPadding=<ltrb 12 2.5 12 0.5>
     pressedSlateSound {resourceObject=null}
    }}
   slot {class=HorizontalBoxSlot, properties {verticalAlignment=VAlign_Center, size {sizeRule=Fill, value=1}}}
  }
  ScaleBox_7 {widgetClassPath=&ScaleBox, parent=btn_ShowAll, widget=ScaleBox_7, slot {class=ButtonSlot}}
  txt_ShowAll {
   widgetClassPath=&TextBlock
   parent=ScaleBox_7
   widget=txt_ShowAll
   properties {
    text=Show All
    visibility=SelfHitTestInvisible
    colorAndOpacity {colorUseRule=UseColor_Specified, specifiedColor=<rgba 0 0 0 1>}
    font {
     bForceMonospaced=false
     bMaterialIsStencil=false
     fontMaterial=null
     fontObject=&Roboto
     letterSpacing=0
     monospacedWidth=1
     size=10.5
     skewAmount=0
     typefaceFontName=Bold
     outlineSettings {
      bApplyOutlineToDropShadows=false
      bMiteredCorners=false
      bSeparateFillAlpha=false
      outlineMaterial=null
      outlineSize=0
      outlineColor=<rgba 0 0 0 1>
     }
    }
   }
   slot {class=ScaleBoxSlot}
  }
  Overlay {
   widgetClassPath=&Overlay
   parent=HorizontalBox_3
   widget=Overlay
   slot {class=HorizontalBoxSlot, properties {verticalAlignment=VAlign_Center, padding=<ltrb 10 0 0 0>, size {sizeRule=Fill, value=1}}}
  }
  Image {
   widgetClassPath=&Image
   parent=Overlay
   widget=Image
   properties {brush {
     bIsDynamicallyLoaded=false
     drawAs=Box
     imageType=NoImage
     mirroring=NoMirror
     resourceName=None
     resourceObject=null
     tiling=NoTile
     imageSize=<32 15.7056>
     margin=<ltrb 0 0 0 0>
     outlineSettings {
      bUseBrushTransparency=false
      roundingType=HalfHeightRadius
      width=0
      color {colorUseRule=UseColor_Specified, specifiedColor=<rgba 0 0 0 0>}
      cornerRadii=<0 0 0 0>
     }
     tintColor {colorUseRule=UseColor_Specified, specifiedColor=<rgba 0.0052 0.0052 0.0052 1>}
     uVRegion {bIsValid=false, max=<0 0>, min=<0 0>}
    }}
   slot {class=OverlaySlot, properties {horizontalAlignment=HAlign_Fill, verticalAlignment=VAlign_Center}}
  }
  SetNameInput {
   widgetClassPath=&EditableText
   bIsVariable=true
   parent=Overlay
   widget=SetNameInput
   properties {
    hintText=AuraCop
    justification=Center
    toolTipText="Set name to filter, the more you type the better. Could be Aura, AuraC, AuraCop.\r\nNot case sensitive"
    widgetStyle {
     backgroundImageComposing {
      bIsDynamicallyLoaded=false
      drawAs=Border
      imageType=FullColor
      mirroring=NoMirror
      resourceName=../../../Engine/Content/Slate/Old/HyperlinkDotted.png
      resourceObject=null
      tiling=Both
      imageSize=<0 0>
      margin=<ltrb 0 0 0 0.1875>
      outlineSettings {
       bUseBrushTransparency=false
       roundingType=HalfHeightRadius
       width=0
       color {colorUseRule=UseColor_Specified, specifiedColor=<rgba 0 0 0 0>}
       cornerRadii=<0 0 0 0>
      }
      tintColor {colorUseRule=UseColor_Specified, specifiedColor=<rgba 1 1 1 1>}
      uVRegion {bIsValid=false, max=<0 0>, min=<0 0>}
     }
     backgroundImageSelected {
      bIsDynamicallyLoaded=false
      drawAs=Image
      imageType=NoImage
      mirroring=NoMirror
      resourceName=None
      resourceObject=null
      tiling=NoTile
      imageSize=<0 0>
      margin=<ltrb 0 0 0 0>
      outlineSettings {
       bUseBrushTransparency=false
       roundingType=HalfHeightRadius
       width=0
       color {colorUseRule=UseColor_Specified, specifiedColor=<rgba 0 0 0 0>}
       cornerRadii=<0 0 0 0>
      }
      tintColor {colorUseRule=UseColor_Specified, specifiedColor=<rgba 0 0.162 0.7454 1>}
      uVRegion {bIsValid=false, max=<0 0>, min=<0 0>}
     }
     caretImage {
      bIsDynamicallyLoaded=false
      drawAs=Image
      imageType=NoImage
      mirroring=NoMirror
      resourceName=None
      resourceObject=null
      tiling=NoTile
      imageSize=<0 0>
      margin=<ltrb 0 0 0 0>
      outlineSettings {
       bUseBrushTransparency=false
       roundingType=HalfHeightRadius
       width=0
       color {colorUseRule=UseColor_Specified, specifiedColor=<rgba 0 0 0 0>}
       cornerRadii=<0 0 0 0>
      }
      tintColor {colorUseRule=UseColor_Specified, specifiedColor=<rgba 1 1 1 1>}
      uVRegion {bIsValid=false, max=<0 0>, min=<0 0>}
     }
     colorAndOpacity {colorUseRule=UseColor_Specified, specifiedColor=<rgba 0.2051 0.5647 0.0262 1>}
     font {
      bForceMonospaced=false
      bMaterialIsStencil=false
      fontMaterial=null
      fontObject=null
      letterSpacing=0
      monospacedWidth=1
      size=10.5
      skewAmount=0
      typefaceFontName=Regular
      outlineSettings {
       bApplyOutlineToDropShadows=false
       bMiteredCorners=false
       bSeparateFillAlpha=false
       outlineMaterial=null
       outlineSize=0
       outlineColor=<rgba 0 0 0 1>
      }
     }
     hintColorAndOpacity {colorUseRule=UseColor_Foreground, specifiedColor=<rgba 1 0 1 1>}
     selectedBackgroundColor {colorUseRule=UseColor_Foreground, specifiedColor=<rgba 1 0 1 1>}
    }
   }
   slot {class=OverlaySlot, properties {horizontalAlignment=HAlign_Fill, verticalAlignment=VAlign_Center}}
  }
  Spacer_7 {
   widgetClassPath=&Spacer
   parent=SeasonTool_VerticalBox
   widget=Spacer_7
   properties {size=<1 12>}
   slot {class=VerticalBoxSlot, properties {verticalAlignment=VAlign_Center}}
  }
  HorizontalBox_161 {
   widgetClassPath=&HorizontalBox
   parent=SeasonTool_VerticalBox
   widget=HorizontalBox_161
   slot {class=VerticalBoxSlot, properties {padding=<ltrb 4 0 4 0>}}
  }
  btn_RenderAssets {
   widgetClassPath=&EditorUtilityButton
   parent=HorizontalBox_161
   widget=btn_RenderAssets
   properties {widgetStyle {
     clickedSlateSound {resourceObject=null}
     disabled {
      bIsDynamicallyLoaded=false
      drawAs=RoundedBox
      imageType=NoImage
      mirroring=NoMirror
      resourceName=None
      resourceObject=null
      tiling=NoTile
      imageSize=<0 0>
      margin=<ltrb 0 0 0 0>
      outlineSettings {
       bUseBrushTransparency=false
       roundingType=FixedRadius
       width=1
       color {colorUseRule=UseColor_ColorTable, specifiedColor=<rgba 0 0 0 0>}
       cornerRadii=<4 4 4 4>
      }
      tintColor {colorUseRule=UseColor_Specified, specifiedColor=<rgba 0.0395 0.0395 0.0395 1>}
      uVRegion {bIsValid=false, max=<0 0>, min=<0 0>}
     }
     disabledForeground {colorUseRule=UseColor_ColorTable, specifiedColor=<rgba 0 0 0 0>}
     hovered {
      bIsDynamicallyLoaded=false
      drawAs=RoundedBox
      imageType=NoImage
      mirroring=NoMirror
      resourceName=None
      resourceObject=null
      tiling=NoTile
      imageSize=<32 32>
      margin=<ltrb 0 0 0 0>
      outlineSettings {
       bUseBrushTransparency=false
       roundingType=FixedRadius
       width=1
       color {colorUseRule=UseColor_ColorTable, specifiedColor=<rgba 0 0 0 0>}
       cornerRadii=<4 4 4 4>
      }
      tintColor {colorUseRule=UseColor_Specified, specifiedColor=<rgba 0.0953 0.0953 0.0953 1>}
      uVRegion {bIsValid=false, max=<0 0>, min=<0 0>}
     }
     hoveredForeground {colorUseRule=UseColor_ColorTable, specifiedColor=<rgba 0 0 0 0>}
     hoveredSlateSound {resourceObject=null}
     normal {
      bIsDynamicallyLoaded=false
      drawAs=RoundedBox
      imageType=NoImage
      mirroring=NoMirror
      resourceName=None
      resourceObject=null
      tiling=NoTile
      imageSize=<32 32>
      margin=<ltrb 0 0 0 0>
      outlineSettings {
       bUseBrushTransparency=false
       roundingType=FixedRadius
       width=1
       color {colorUseRule=UseColor_ColorTable, specifiedColor=<rgba 0 0 0 0>}
       cornerRadii=<4 4 4 4>
      }
      tintColor {colorUseRule=UseColor_Specified, specifiedColor=<rgba 0.5454 0.5647 0.2787 0.2812>}
      uVRegion {bIsValid=false, max=<0 0>, min=<0 0>}
     }
     normalForeground {colorUseRule=UseColor_ColorTable, specifiedColor=<rgba 0 0 0 0>}
     normalPadding=<ltrb 12 1.5 12 1.5>
     pressed {
      bIsDynamicallyLoaded=false
      drawAs=RoundedBox
      imageType=NoImage
      mirroring=NoMirror
      resourceName=None
      resourceObject=null
      tiling=NoTile
      imageSize=<32 32>
      margin=<ltrb 0 0 0 0>
      outlineSettings {
       bUseBrushTransparency=false
       roundingType=FixedRadius
       width=1
       color {colorUseRule=UseColor_ColorTable, specifiedColor=<rgba 0 0 0 0>}
       cornerRadii=<4 4 4 4>
      }
      tintColor {colorUseRule=UseColor_Specified, specifiedColor=<rgba 0.0284 0.0284 0.0284 1>}
      uVRegion {bIsValid=false, max=<0 0>, min=<0 0>}
     }
     pressedForeground {colorUseRule=UseColor_ColorTable, specifiedColor=<rgba 0 0 0 0>}
     pressedPadding=<ltrb 12 2.5 12 0.5>
     pressedSlateSound {resourceObject=null}
    }}
   slot {class=HorizontalBoxSlot}
  }
  ScaleBox_1 {widgetClassPath=&ScaleBox, parent=btn_RenderAssets, widget=ScaleBox_1, slot {class=ButtonSlot}}
  txt_RenderAssets {
   widgetClassPath=&TextBlock
   parent=ScaleBox_1
   widget=txt_RenderAssets
   properties {
    text="(WIP) Render Assets"
    visibility=SelfHitTestInvisible
    colorAndOpacity {colorUseRule=UseColor_Specified, specifiedColor=<rgba 0 0 0 1>}
    font {
     bForceMonospaced=false
     bMaterialIsStencil=false
     fontMaterial=null
     fontObject=&Roboto
     letterSpacing=0
     monospacedWidth=1
     size=10.5
     skewAmount=0
     typefaceFontName=Bold
     outlineSettings {
      bApplyOutlineToDropShadows=false
      bMiteredCorners=false
      bSeparateFillAlpha=false
      outlineMaterial=null
      outlineSize=0
      outlineColor=<rgba 0 0 0 1>
     }
    }
   }
   slot {class=ScaleBoxSlot}
  }
  btn_Debug {
   widgetClassPath=&EditorUtilityButton
   parent=HorizontalBox_161
   widget=btn_Debug
   properties {widgetStyle {
     clickedSlateSound {resourceObject=null}
     disabled {
      bIsDynamicallyLoaded=false
      drawAs=RoundedBox
      imageType=NoImage
      mirroring=NoMirror
      resourceName=None
      resourceObject=null
      tiling=NoTile
      imageSize=<0 0>
      margin=<ltrb 0 0 0 0>
      outlineSettings {
       bUseBrushTransparency=false
       roundingType=FixedRadius
       width=1
       color {colorUseRule=UseColor_ColorTable, specifiedColor=<rgba 0 0 0 0>}
       cornerRadii=<4 4 4 4>
      }
      tintColor {colorUseRule=UseColor_Specified, specifiedColor=<rgba 0.0395 0.0395 0.0395 1>}
      uVRegion {bIsValid=false, max=<0 0>, min=<0 0>}
     }
     disabledForeground {colorUseRule=UseColor_ColorTable, specifiedColor=<rgba 0 0 0 0>}
     hovered {
      bIsDynamicallyLoaded=false
      drawAs=RoundedBox
      imageType=NoImage
      mirroring=NoMirror
      resourceName=None
      resourceObject=null
      tiling=NoTile
      imageSize=<32 32>
      margin=<ltrb 0 0 0 0>
      outlineSettings {
       bUseBrushTransparency=false
       roundingType=FixedRadius
       width=1
       color {colorUseRule=UseColor_ColorTable, specifiedColor=<rgba 0 0 0 0>}
       cornerRadii=<4 4 4 4>
      }
      tintColor {colorUseRule=UseColor_Specified, specifiedColor=<rgba 0.0953 0.0953 0.0953 1>}
      uVRegion {bIsValid=false, max=<0 0>, min=<0 0>}
     }
     hoveredForeground {colorUseRule=UseColor_ColorTable, specifiedColor=<rgba 0 0 0 0>}
     hoveredSlateSound {resourceObject=null}
     normal {
      bIsDynamicallyLoaded=false
      drawAs=RoundedBox
      imageType=NoImage
      mirroring=NoMirror
      resourceName=None
      resourceObject=null
      tiling=NoTile
      imageSize=<32 32>
      margin=<ltrb 0 0 0 0>
      outlineSettings {
       bUseBrushTransparency=false
       roundingType=FixedRadius
       width=1
       color {colorUseRule=UseColor_ColorTable, specifiedColor=<rgba 0 0 0 0>}
       cornerRadii=<4 4 4 4>
      }
      tintColor {colorUseRule=UseColor_Specified, specifiedColor=<rgba 0.5454 0.5647 0.2787 0.2812>}
      uVRegion {bIsValid=false, max=<0 0>, min=<0 0>}
     }
     normalForeground {colorUseRule=UseColor_ColorTable, specifiedColor=<rgba 0 0 0 0>}
     normalPadding=<ltrb 12 1.5 12 1.5>
     pressed {
      bIsDynamicallyLoaded=false
      drawAs=RoundedBox
      imageType=NoImage
      mirroring=NoMirror
      resourceName=None
      resourceObject=null
      tiling=NoTile
      imageSize=<32 32>
      margin=<ltrb 0 0 0 0>
      outlineSettings {
       bUseBrushTransparency=false
       roundingType=FixedRadius
       width=1
       color {colorUseRule=UseColor_ColorTable, specifiedColor=<rgba 0 0 0 0>}
       cornerRadii=<4 4 4 4>
      }
      tintColor {colorUseRule=UseColor_Specified, specifiedColor=<rgba 0.0284 0.0284 0.0284 1>}
      uVRegion {bIsValid=false, max=<0 0>, min=<0 0>}
     }
     pressedForeground {colorUseRule=UseColor_ColorTable, specifiedColor=<rgba 0 0 0 0>}
     pressedPadding=<ltrb 12 2.5 12 0.5>
     pressedSlateSound {resourceObject=null}
    }}
   slot {class=HorizontalBoxSlot}
  }
  ScaleBox_6 {widgetClassPath=&ScaleBox, parent=btn_Debug, widget=ScaleBox_6, slot {class=ButtonSlot}}
  txt_Debug {
   widgetClassPath=&TextBlock
   parent=ScaleBox_6
   widget=txt_Debug
   properties {
    text=Debug
    visibility=SelfHitTestInvisible
    colorAndOpacity {colorUseRule=UseColor_Specified, specifiedColor=<rgba 0 0 0 1>}
    font {
     bForceMonospaced=false
     bMaterialIsStencil=false
     fontMaterial=null
     fontObject=&Roboto
     letterSpacing=0
     monospacedWidth=1
     size=10.5
     skewAmount=0
     typefaceFontName=Bold
     outlineSettings {
      bApplyOutlineToDropShadows=false
      bMiteredCorners=false
      bSeparateFillAlpha=false
      outlineMaterial=null
      outlineSize=0
      outlineColor=<rgba 0 0 0 1>
     }
    }
   }
   slot {class=ScaleBoxSlot}
  }
 ]
}
