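# Editor Utility Widget hub for the Season Tool, bundling the asset spawner, day light, art review visualizer, cosmetic picker and console command panels for artists.
# The five panels are embedded in one overlay and cached in an array on Construct.
# Tab buttons call ShowTab, which hides every panel and then shows the chosen one.
# Right-clicking a button's border spawns that tool as a separate editor tab via EditorUtilitySubsystem.
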EditorUtilityWidgetBlueprint /Game/ArtTools/HS_Season_Tool/Season_Tool_Main
graphs [
 CheckRightClick {nodes [
   n1 = CheckRightClick!() -> (then, Input:"Pointer Event Structure (by ref)", EUW_BP_Ref:Editor Utility Widget Blueprint Object Reference) {is=title K2Node_FunctionEntry}
   > Branch(Condition="Equal (Key)"(
     A:Key Structure=Get Effecting Button!(Input:"Pointer Event Structure (by ref)"=@n1) -> (ReturnValue:Key Structure) {is=title K2Node_CallFunction}
     B:Key Structure=RightMouseButton
    ) -> (ReturnValue:bool=~) {is=title K2Node_PromotableOperator})
   > SpawnAndRegisterTab(self=Get EditorUtilitySubsystem(), InBlueprint=@n1.EUW_BP_Ref)
  ]}
 EventGraph {nodes [
   "On Clicked (btn_ConsoleCommandUI)"() {is=title K2Node_ComponentBoundEvent}
   > ShowTab!(
    execute
    self
    EUW_Ref:Editor Utility Widget Object Reference=Get ConsoleCommandUI() -> (ConsoleCommandUI:Console Command UI Object Reference)
   ) -> (then)
   "On Clicked (btn_SeasonAssetsSpawner)"() {is=title K2Node_ComponentBoundEvent}
   > ShowTab(EUW_Ref=Get Season_Tool_SpawnAsset_Panel() -> (Season_Tool_SpawnAsset_Panel:Season Tool Spawn Asset Panel Object Reference))
   "On Clicked (btn_ArtReviewVisualizer)"() {is=title K2Node_ComponentBoundEvent}
   > ShowTab(EUW_Ref=Get ArtReview_Visualizer() -> (ArtReview_Visualizer:Art Review Visualizer Object Reference))
   "On Clicked (btn_DayLight)"() {is=title K2Node_ComponentBoundEvent}
   > ShowTab(EUW_Ref=Get Season_Tool_DayLight_Panel() -> (Season_Tool_DayLight_Panel:Season Tool Day Light Panel Object Reference))
   "On Clicked (btn_CosmeticPicker)"() {is=title K2Node_ComponentBoundEvent}
   > ShowTab(EUW_Ref=Get CosmeticPickerv2() -> (CosmeticPickerv2:Cosmetic Pickerv 2 Object Reference))
   Event Construct()
   > n1 = For Each Loop(Array:Array of Editor Utility Widget Object References=Make Array!(
     "[0]":Editor Utility Widget Object Reference=Get Season_Tool_DayLight_Panel()
     "[1]":Editor Utility Widget Object Reference=Get Season_Tool_SpawnAsset_Panel()
     "[2]":Editor Utility Widget Object Reference=Get ArtReview_Visualizer()
     "[3]":Editor Utility Widget Object Reference=Get CosmeticPickerv2()
     "[4]":Editor Utility Widget Object Reference=Get ConsoleCommandUI()
    ) -> (Array:Array of Editor Utility Widget Object References)) -> (Array Element:Editor Utility Widget Object Reference)
   > Add(
    TargetArray:Array of Editor Utility Widget Object References=Get EditorUtilityWidgetArray() -> (EditorUtilityWidgetArray:Array of Editor Utility Widget Object References)
    NewItem:"Editor Utility Widget Object Reference (by ref)"=@n1
   ) {is=title K2Node_CallArrayFunction}
  ]}
 On_ArtReviewVisualizer_Border_MouseButtonDown {nodes [
   n1 = On_ArtReviewVisualizer_Border_MouseButtonDown!() -> (then, MyGeometry:Geometry Structure, MouseEvent:"Pointer Event Structure (by ref)") {is=title K2Node_FunctionEntry}
   > CheckRightClick!(
    execute
    self
    Input:"Pointer Event Structure (by ref)"=@n1.MouseEvent
    EUW_BP_Ref:Editor Utility Widget Blueprint Object Reference=/Game/NeverCook/ArtTools/Tech_Art/ArtReview_Visualizer/ArtReview_Visualizer
   ) -> (then) {is=title K2Node_CallFunction}
   > Return Node!(execute, ReturnValue:Event Reply Structure=Handled())
  ]}
 On_ConsoleCommandUI_Border_MouseButtonDown {nodes [
   n1 = On_ConsoleCommandUI_Border_MouseButtonDown!() -> (then, MyGeometry:Geometry Structure, MouseEvent:"Pointer Event Structure (by ref)") {is=title K2Node_FunctionEntry}
   > CheckRightClick(Input=@n1.MouseEvent, EUW_BP_Ref=/Game/ArtTools/Tech_Anim/ConsoleCommandsUI/ConsoleCommandUI)
   > Return Node(ReturnValue=Handled())
  ]}
 On_CosmeticPicker_Broder_MouseButtonDown {nodes [
   n1 = On_CosmeticPicker_Broder_MouseButtonDown!() -> (then, MyGeometry:Geometry Structure, MouseEvent:"Pointer Event Structure (by ref)") {is=title K2Node_FunctionEntry}
   > CheckRightClick(Input=@n1.MouseEvent, EUW_BP_Ref=/Game/NeverCook/ArtTools/Tech_Art/CosmeticPicker/CosmeticPickerv2)
   > Return Node(ReturnValue=Handled())
  ]}
 On_DayLight_Border_MouseButtonDown {nodes [
   n1 = On_DayLight_Border_MouseButtonDown!() -> (then, MyGeometry:Geometry Structure, MouseEvent:"Pointer Event Structure (by ref)") {is=title K2Node_FunctionEntry}
   > CheckRightClick(Input=@n1.MouseEvent, EUW_BP_Ref=&/Season_Tool_DayLight_Panel)
   > Return Node(ReturnValue=Handled())
  ]}
 On_SeasonAssetsSpawner_Border_MouseButtonDown {nodes [
   n1 = On_SeasonAssetsSpawner_Border_MouseButtonDown!() -> (then, MyGeometry:Geometry Structure, MouseEvent:"Pointer Event Structure (by ref)") {is=title K2Node_FunctionEntry}
   > CheckRightClick(Input=@n1.MouseEvent, EUW_BP_Ref=&/Season_Tool_SpawnAsset_Panel)
   > Return Node(ReturnValue=Handled())
  ]}
 ShowTab {nodes [
   n1 = ShowTab!() -> (then, EUW_Ref:Editor Utility Widget Object Reference) {is=title K2Node_FunctionEntry}
   > n2 = For Each Loop(Array=Get EditorUtilityWidgetArray())
   SetVisibility(execute=@n2.Completed, self=@n1, InVisibility=Visible)
   Branch(
    execute=@n2.LoopBody
    Condition="Not Equal (Enum)"!(
     A:ESlateVisibility Enum=GetVisibility!(self:Widget Object Reference) -> (ReturnValue:ESlateVisibility Enum=Visible)
     B:ESlateVisibility Enum=Hidden
    ) -> (ReturnValue:bool)
   )
   > SetVisibility(self=@n2, InVisibility=Hidden)
  ]}
]
variables [EditorUtilityWidgetArray {type="array<EditorUtilityWidget>", default []}]
widgets {
 info {rootWidgetClass=&StackBox, widgetCount=34}
 widgets [
  StackBox_78 {widgetClassPath=&StackBox, widget=StackBox_78, properties {bOverride_Cursor=true}}
  VerticalBox_0 {
   widgetClassPath=&VerticalBox
   parent=StackBox_78
   widget=VerticalBox_0
   slot {class=StackBoxSlot, properties {padding=<ltrb 5 0 5 0>, size {sizeRule=Fill, value=1}}}
  }
  Tabs_HorizontalBox {
   widgetClassPath=&HorizontalBox
   parent=VerticalBox_0
   widget=Tabs_HorizontalBox
   slot {class=VerticalBoxSlot, properties {padding=<ltrb 0 6 0 6>}}
  }
  UniformGridPanel_0 {
   widgetClassPath=/Script/UMG.UniformGridPanel
   parent=Tabs_HorizontalBox
   widget=UniformGridPanel_0
   properties {slotPadding=<ltrb 4 4 4 4>}
   slot {class=HorizontalBoxSlot, properties {padding=<ltrb 3 0 3 0>, size {sizeRule=Fill, value=1}}}
  }
  SeasonAssetsSpawner_Border {
   widgetClassPath=&Border
   parent=UniformGridPanel_0
   widget=SeasonAssetsSpawner_Border
   properties {brushColor=<rgba 1 1 1 0>, padding=<ltrb 0 0 0 0>}
   slot {class=UniformGridSlot, properties {horizontalAlignment=HAlign_Fill}}
  }
  btn_SeasonAssetsSpawner {
   widgetClassPath=&EditorUtilityButton
   parent=SeasonAssetsSpawner_Border
   widget=btn_SeasonAssetsSpawner
   slot {class=BorderSlot, properties {padding=<ltrb 0 0 0 0>}}
  }
  ScaleBox_1 {
   widgetClassPath=&ScaleBox
   parent=btn_SeasonAssetsSpawner
   widget=ScaleBox_1
   slot {class=ButtonSlot, properties {horizontalAlignment=HAlign_Fill, verticalAlignment=VAlign_Fill, padding=<ltrb 2 2 2 2>}}
  }
  txt_SpawnSeasonAssets_1 {
   widgetClassPath=&TextBlock
   parent=ScaleBox_1
   widget=txt_SpawnSeasonAssets_1
   properties {
    justification=Center
    text=Assets Spawner
    visibility=SelfHitTestInvisible
    font {
     bForceMonospaced=false
     bMaterialIsStencil=false
     fontMaterial=null
     fontObject=&NotoSans
     letterSpacing=0
     monospacedWidth=1
     size=13.5
     skewAmount=0
     typefaceFontName=NotoSans-Regular
     outlineSettings {
      bApplyOutlineToDropShadows=false
      bMiteredCorners=false
      bSeparateFillAlpha=false
      outlineMaterial=null
      outlineSize=0
      outlineColor=<rgba 0 0 0 1>
     }
    }
   }
   slot {class=ScaleBoxSlot, properties {horizontalAlignment=HAlign_Fill}}
  }
  DayLight_Border {
   widgetClassPath=&Border
   parent=UniformGridPanel_0
   widget=DayLight_Border
   properties {brushColor=<rgba 1 1 1 0>, padding=<ltrb 0 0 0 0>}
   slot {class=UniformGridSlot, properties {column=1, horizontalAlignment=HAlign_Fill}}
  }
  btn_DayLight {
   widgetClassPath=&EditorUtilityButton
   parent=DayLight_Border
   widget=btn_DayLight
   slot {class=BorderSlot, properties {padding=<ltrb 0 0 0 0>}}
  }
  ScaleBox_4 {
   widgetClassPath=&ScaleBox
   parent=btn_DayLight
   widget=ScaleBox_4
   slot {class=ButtonSlot, properties {horizontalAlignment=HAlign_Fill, verticalAlignment=VAlign_Fill, padding=<ltrb 2 2 2 2>}}
  }
  txt_SpawnSeasonAssets_2 {
   widgetClassPath=&TextBlock
   parent=ScaleBox_4
   widget=txt_SpawnSeasonAssets_2
   properties {
    justification=Center
    text=Day Light
    visibility=SelfHitTestInvisible
    font {
     bForceMonospaced=false
     bMaterialIsStencil=false
     fontMaterial=null
     fontObject=&NotoSans
     letterSpacing=0
     monospacedWidth=1
     size=13.5
     skewAmount=0
     typefaceFontName=NotoSans-Regular
     outlineSettings {
      bApplyOutlineToDropShadows=false
      bMiteredCorners=false
      bSeparateFillAlpha=false
      outlineMaterial=null
      outlineSize=0
      outlineColor=<rgba 0 0 0 1>
     }
    }
   }
   slot {class=ScaleBoxSlot, properties {horizontalAlignment=HAlign_Fill}}
  }
  ArtReviewVisualizer_Border {
   widgetClassPath=&Border
   parent=UniformGridPanel_0
   widget=ArtReviewVisualizer_Border
   properties {brushColor=<rgba 1 1 1 0>, padding=<ltrb 0 0 0 0>}
   slot {class=UniformGridSlot, properties {column=2, horizontalAlignment=HAlign_Fill}}
  }
  btn_ArtReviewVisualizer {
   widgetClassPath=&EditorUtilityButton
   parent=ArtReviewVisualizer_Border
   widget=btn_ArtReviewVisualizer
   slot {class=BorderSlot, properties {padding=<ltrb 0 0 0 0>}}
  }
  ScaleBox_2 {
   widgetClassPath=&ScaleBox
   parent=btn_ArtReviewVisualizer
   widget=ScaleBox_2
   slot {class=ButtonSlot, properties {horizontalAlignment=HAlign_Fill, verticalAlignment=VAlign_Fill, padding=<ltrb 2 2 2 2>}}
  }
  txt_ArtReviewVisualizer {
   widgetClassPath=&TextBlock
   parent=ScaleBox_2
   widget=txt_ArtReviewVisualizer
   properties {
    justification=Center
    text=Art Review Visualizer
    visibility=SelfHitTestInvisible
    font {
     bForceMonospaced=false
     bMaterialIsStencil=false
     fontMaterial=null
     fontObject=&NotoSans
     letterSpacing=0
     monospacedWidth=1
     size=13.5
     skewAmount=0
     typefaceFontName=NotoSans-Regular
     outlineSettings {
      bApplyOutlineToDropShadows=false
      bMiteredCorners=false
      bSeparateFillAlpha=false
      outlineMaterial=null
      outlineSize=0
      outlineColor=<rgba 0 0 0 1>
     }
    }
   }
   slot {class=ScaleBoxSlot, properties {horizontalAlignment=HAlign_Fill}}
  }
  CosmeticPicker_Border {
   widgetClassPath=&Border
   parent=UniformGridPanel_0
   widget=CosmeticPicker_Border
   properties {brushColor=<rgba 1 1 1 0>, padding=<ltrb 0 0 0 0>}
   slot {class=UniformGridSlot, properties {horizontalAlignment=HAlign_Fill, row=1}}
  }
  btn_CosmeticPicker {
   widgetClassPath=&EditorUtilityButton
   parent=CosmeticPicker_Border
   widget=btn_CosmeticPicker
   slot {class=BorderSlot, properties {padding=<ltrb 0 0 0 0>}}
  }
  ScaleBox_3 {
   widgetClassPath=&ScaleBox
   parent=btn_CosmeticPicker
   widget=ScaleBox_3
   slot {class=ButtonSlot, properties {horizontalAlignment=HAlign_Fill, verticalAlignment=VAlign_Fill, padding=<ltrb 2 2 2 2>}}
  }
  txt_CosmeticPicker {
   widgetClassPath=&TextBlock
   parent=ScaleBox_3
   widget=txt_CosmeticPicker
   properties {
    justification=Center
    text=Cosmetic Picker v2
    visibility=SelfHitTestInvisible
    font {
     bForceMonospaced=false
     bMaterialIsStencil=false
     fontMaterial=null
     fontObject=&NotoSans
     letterSpacing=0
     monospacedWidth=1
     size=13.5
     skewAmount=0
     typefaceFontName=NotoSans-Regular
     outlineSettings {
      bApplyOutlineToDropShadows=false
      bMiteredCorners=false
      bSeparateFillAlpha=false
      outlineMaterial=null
      outlineSize=0
      outlineColor=<rgba 0 0 0 1>
     }
    }
   }
   slot {class=ScaleBoxSlot, properties {horizontalAlignment=HAlign_Fill}}
  }
  ConsoleCommandUI_Border {
   widgetClassPath=&Border
   parent=UniformGridPanel_0
   widget=ConsoleCommandUI_Border
   properties {brushColor=<rgba 1 1 1 0>, padding=<ltrb 0 0 0 0>}
   slot {class=UniformGridSlot, properties {column=1, horizontalAlignment=HAlign_Fill, row=1}}
  }
  btn_ConsoleCommandUI {
   widgetClassPath=&EditorUtilityButton
   parent=ConsoleCommandUI_Border
   widget=btn_ConsoleCommandUI
   slot {class=BorderSlot, properties {padding=<ltrb 0 0 0 0>}}
  }
  ScaleBox {
   widgetClassPath=&ScaleBox
   parent=btn_ConsoleCommandUI
   widget=ScaleBox
   slot {class=ButtonSlot, properties {horizontalAlignment=HAlign_Fill, verticalAlignment=VAlign_Fill, padding=<ltrb 2 2 2 2>}}
  }
  txt_CosmeticPicker_1 {
   widgetClassPath=&TextBlock
   parent=ScaleBox
   widget=txt_CosmeticPicker_1
   properties {
    justification=Center
    text=Console Command UI
    visibility=SelfHitTestInvisible
    font {
     bForceMonospaced=false
     bMaterialIsStencil=false
     fontMaterial=null
     fontObject=&NotoSans
     letterSpacing=0
     monospacedWidth=1
     size=13.5
     skewAmount=0
     typefaceFontName=NotoSans-Regular
     outlineSettings {
      bApplyOutlineToDropShadows=false
      bMiteredCorners=false
      bSeparateFillAlpha=false
      outlineMaterial=null
      outlineSize=0
      outlineColor=<rgba 0 0 0 1>
     }
    }
   }
   slot {class=ScaleBoxSlot, properties {horizontalAlignment=HAlign_Fill}}
  }
  ScaleBox_0 {widgetClassPath=&ScaleBox, parent=VerticalBox_0, widget=ScaleBox_0, slot {class=VerticalBoxSlot}}
  text_Title_SeasonTool {
   widgetClassPath=&TextBlock
   parent=ScaleBox_0
   widget=text_Title_SeasonTool
   properties {
    justification=Center
    text="Season Tool (Proto version: v1.1.1)"
    font {
     bForceMonospaced=false
     bMaterialIsStencil=false
     fontMaterial=null
     fontObject=&NotoSans
     letterSpacing=0
     monospacedWidth=1
     size=9
     skewAmount=0
     typefaceFontName=NotoSans-Regular
     outlineSettings {
      bApplyOutlineToDropShadows=false
      bMiteredCorners=false
      bSeparateFillAlpha=false
      outlineMaterial=null
      outlineSize=0
      outlineColor=<rgba 0 0 0 1>
     }
    }
   }
   slot {class=ScaleBoxSlot, properties {horizontalAlignment=HAlign_Fill}}
  }
  Spacer {widgetClassPath=&Spacer, parent=VerticalBox_0, widget=Spacer, properties {size=<1 5>}, slot {class=VerticalBoxSlot}}
  EditorUtilityScrollBox_0 {
   widgetClassPath=&EditorUtilityScrollBox
   parent=VerticalBox_0
   widget=EditorUtilityScrollBox_0
   slot {class=VerticalBoxSlot, properties {size {sizeRule=Fill, value=1}}}
  }
  Overlay_0 {widgetClassPath=&Overlay, parent=EditorUtilityScrollBox_0, widget=Overlay_0, slot {class=ScrollBoxSlot}}
  Season_Tool_SpawnAsset_Panel {
   widgetClassPath=&Season_Tool_SpawnAsset_Panel
   bIsVariable=true
   parent=Overlay_0
   widget=Season_Tool_SpawnAsset_Panel
   slot {class=OverlaySlot, properties {horizontalAlignment=HAlign_Fill, verticalAlignment=VAlign_Fill}}
  }
  Season_Tool_DayLight_Panel {
   widgetClassPath=&Season_Tool_DayLight_Panel
   bIsVariable=true
   parent=Overlay_0
   widget=Season_Tool_DayLight_Panel
   properties {visibility=Hidden}
   slot {class=OverlaySlot, properties {horizontalAlignment=HAlign_Fill, verticalAlignment=VAlign_Fill}}
  }
  CosmeticPickerv2 {
   widgetClassPath=&CosmeticPickerv2
   bIsVariable=true
   parent=Overlay_0
   widget=CosmeticPickerv2
   properties {visibility=Hidden}
   slot {class=OverlaySlot, properties {horizontalAlignment=HAlign_Fill, verticalAlignment=VAlign_Fill}}
  }
  ArtReview_Visualizer {
   widgetClassPath=&ArtReview_Visualizer
   bIsVariable=true
   parent=Overlay_0
   widget=ArtReview_Visualizer
   properties {visibility=Hidden}
   slot {class=OverlaySlot, properties {horizontalAlignment=HAlign_Fill, verticalAlignment=VAlign_Fill}}
  }
  ConsoleCommandUI {
   widgetClassPath=&ConsoleCommandUI
   bIsVariable=true
   parent=Overlay_0
   widget=ConsoleCommandUI
   properties {visibility=Hidden}
   slot {class=OverlaySlot, properties {horizontalAlignment=HAlign_Fill, verticalAlignment=VAlign_Fill}}
  }
 ]
}
